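# Creative-mode modal widget, built on FortCreativeRichTextEditorWidget, for editing and styling a device's rich text field.
# A tabbed switcher separates raw text entry from styling of selected words, with button visibility following selection and gamepad input.
# Devices implementing BPi_TextStyleOverrides can replace the style table, and a warning animation flags the character limit.
# Save and cancel pass through confirmation dialogs, whose results fire OnTextChangeConfirmed and close the modal.
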WidgetBlueprint /FortniteCoreUI/InGame/Creative/PropertyEditor/StyleEditor/TextAndStyleEditor
animations [Ani_Warning {
  binding_count=3
  length_seconds=0.25
  object=&self:Ani_Warning
  sequence {bindings [
    Border_CharWarning {
     id=#1
     possessed_class=Border
     tracks [{
       type=MovieSceneColorTrack
       display_name=Brush Color
       sections [{
         type=MovieSceneColorSection
         end_frame=5
         start_frame=0
         channels [
          Color.R {type=MovieSceneScriptingFloatChannel, keys [[0,1],[5,0.7922]]}
          Color.G {type=MovieSceneScriptingFloatChannel, keys [[0,1],[5,0.0078]]}
          Color.B {type=MovieSceneScriptingFloatChannel, keys [[0,1],[5,0.1843]]}
          Color.A {type=MovieSceneScriptingFloatChannel, const [1,2]}
         ]
        }]
      }]
    }
    Text_CharCount {
     id=#2
     possessed_class=CommonTextBlock
     tracks [{
       type=MovieSceneColorTrack
       display_name=Color and Opacity
       sections [{
         type=MovieSceneColorSection
         end_frame=5
         start_frame=0
         channels [
          Color.R {type=MovieSceneScriptingFloatChannel, keys [[0,0.0296],[5,0.7922]]}
          Color.G {type=MovieSceneScriptingFloatChannel, keys [[0,1],[5,0.0078]]}
          Color.B {type=MovieSceneScriptingFloatChannel, keys [[0,0.2232],[5,0.1843]]}
          Color.A {type=MovieSceneScriptingFloatChannel, const [1,2]}
         ]
        }]
      }]
    }
    WarningImage {
     id=#3
     possessed_class=Overlay
     tracks [
      {
       type=MovieSceneFloatTrack
       display_name=Render Opacity
       sections [{type=MovieSceneFloatSection, end_frame=5, start_frame=0, channels [None {type=MovieSceneScriptingFloatChannel, keys [[0,0],[5,1]]}]}]
      }
      {
       type=MovieSceneByteTrack
       display_name=Visibility
       sections [{type=MovieSceneByteSection, end_frame=0, start_frame=0, channels [None {type=MovieSceneScriptingByteChannel, const [0,1]}]}]
      }
     ]
    }
   ]}
 }]
graphs [
 Apply Button Visibility {nodes [
   Apply Button Visibility() {is=title K2Node_FunctionEntry}
   > n1 = Get Widget Context Visibility!(execute, self) -> (then, IsVisible:bool) {is=title K2Node_CallFunction}
   > Set IsWidgetContextVisible(IsWidgetContextVisible:bool=@n1) -> (Output_Get:bool)
   > n2 = GetButtonContext!(execute, self:Fort Creative Rich Text Editor Widget Object Reference) -> (then, ReturnValue:EButtonContext Enum=NoButtonSelected)
   > n3 = Switch on EButtonContext!(execute, Selection:EButtonContext Enum=@n2) -> (NoButtonSelected:Exec, ButtonSelected:Exec, StylingButtons:Exec) {is=title K2Node_SwitchEnum}
   SetVisibility(
    execute=@n3.ButtonSelected
    self=Get Button_ClearSelection() -> (Button_ClearSelection:Icon Text Button Object Reference)
    InVisibility=SelfHitTestInvisible
   )
   n4 = Get WordButtons_WBox() -> (WordButtons_WBox:Wrap Box Object Reference)
   n5 = Select(
    Option 0=Collapsed
    Option 1=SelfHitTestInvisible
    Index=AND Boolean(
     A=Get IsWidgetContextVisible() -> (IsWidgetContextVisible:bool)
     B=HasAnyChildren!(self:Panel Widget Object Reference=@n4) -> (ReturnValue:bool)
    )
   )
   > SetVisibility(self=Get Button_Style() -> (Button_Style:Icon Text Button Object Reference), InVisibility=@n5)
   > SetVisibility(self=Get Button_WordButtonLabel() -> (Button_WordButtonLabel:Icon Text Button Object Reference), InVisibility=@n5)
   SetVisibility(execute=@n3.NoButtonSelected, self=Get Button_Style(), InVisibility=Collapsed)
   > SetVisibility(self=Get Button_ClearSelection(), InVisibility=Collapsed)
   n6 = HasAnyChildren(self=Get WordButtons_WBox())
   n7 = Get IsWidgetContextVisible()
   > SetVisibility(self=Get Button_WordButtonLabel(), InVisibility=Select(Option 0=Collapsed, Option 1=SelfHitTestInvisible, Index=AND Boolean(A=@n7, B=@n6)))
   SetVisibility(execute=@n3.StylingButtons, self=Get Button_Style(), InVisibility=Collapsed)
   > SetVisibility(self=Get Button_ClearSelection(), InVisibility=Collapsed)
   > SetVisibility(self=Get Button_WordButtonLabel(), InVisibility=Collapsed)
  ]}
 ClearAllText {nodes [
   ClearAllText() {is=title K2Node_FunctionEntry}
   > "SetText (Multi-Line Editable Text)"!(
    execute
    self:"Editable Text (Multi- Line) Object Reference"=Get EditableText() -> (EditableText:"Editable Text (Multi- Line) Object Reference")
    InText:Text=GetEmptyText!() -> (ReturnValue:Text)
   ) -> (then)
  ]}
 EventGraph {nodes [
   "On Button Clicked (Button_Style)"!() -> (OutputDelegate, then, Button:Common Button Legacy Object Reference) {is=title K2Node_ComponentBoundEvent}
   n1 = Get EntryBox_StyleButtons() -> (EntryBox_StyleButtons:Dynamic Entry Box Object Reference)
   > SetFocus(self="Get (a copy)"(Array:Array of User Widget Object References=GetAllEntries!(self:Dynamic Entry Box Base Object Reference=@n1) -> (ReturnValue:Array of User Widget Object References)) -> (Output:User Widget Object Reference))
   > Set bStylingInProgress(bStylingInProgress:bool=true) -> (Output_Get:bool)
   > SetCommitCancelState!(execute, self, OkEnabled:bool=false, CancelEnabled:bool=true) -> (then)
   > Apply Button Visibility() {is=title K2Node_CallFunction}
   "On Button Clicked (Button_Cancel)"!() -> (OutputDelegate, then, Button:Common Button Legacy Object Reference) {is=title K2Node_ComponentBoundEvent}
   > n2 = Branch(Condition=Get bStylingInProgress() -> (bStylingInProgress:bool))
   n3 = Get LastSelectedWordButton() -> (LastSelectedWordButton:Fort Creative Rich Text Word Button Object Reference)
   > n4 = Is Valid(InputObject=@n3)
   > n5 = SetFocus(self=@n3)
   Set bStylingInProgress(execute=[@n5.then, @n4.Is Not Valid], bStylingInProgress=false)
   > SetCommitCancelState(OkEnabled=true, CancelEnabled=true)
   > Apply Button Visibility()
   SetCommitCancelState(execute=@n2.else, OkEnabled=false, CancelEnabled=false)
   > ShowConfirmationDialog!(execute, self:Fort Creative Rich Text Editor Widget Object Reference, DialogType:EConfirmDialogType Enum=CancelChanges) -> (then)
   "On Button Clicked (UserActionMenuInputButton_Left)"!() -> (OutputDelegate, then, Button:Common Button Legacy Object Reference) {is=title K2Node_ComponentBoundEvent}
   n6 = Get TabsButtonGroup() -> (TabsButtonGroup:Common Button Group Base Object Reference)
   > SelectPreviousButton!(execute, self:Common Button Group Base Object Reference=@n6, bAllowWrap:bool=false, bAllowSound:bool=true) -> (then)
   "On Button Clicked (UserActionMenuInputButton_Right)"!() -> (OutputDelegate, then, Button:Common Button Legacy Object Reference) {is=title K2Node_ComponentBoundEvent}
   > SelectNextButton!(execute, self:Common Button Group Base Object Reference=@n6, bAllowWrap:bool=false, bAllowSound:bool=true) -> (then)
   "On Clicked (SaveAndClose_Button)"!() -> (OutputDelegate, then, Button:Common Button Base Object Reference) {is=title K2Node_ComponentBoundEvent}
   > SetCommitCancelState(OkEnabled=false, CancelEnabled=false)
   > ShowConfirmationDialog(DialogType=CharLimitExceeded)
   "On Button Clicked (Button_ClearSelection)"!() -> (OutputDelegate, then, Button:Common Button Legacy Object Reference) {is=title K2Node_ComponentBoundEvent}
   > ClearSelectedWordButtons(self:Fort Creative Rich Text Editor Widget Object Reference)
   > Apply Button Visibility()
   "On Clicked (ClearText_Button)"!() -> (OutputDelegate, then, Button:Common Button Base Object Reference) {is=title K2Node_ComponentBoundEvent}
   > ClearAllText() {is=title K2Node_CallFunction}
   "On Clicked (EditText_Button)"!() -> (OutputDelegate, then, Button:Common Button Base Object Reference) {is=title K2Node_ComponentBoundEvent}
   > SetEditableTextFocus()
   "On Button Clicked (Button_Help)"!() -> (OutputDelegate, then, Button:Common Button Legacy Object Reference) {is=title K2Node_ComponentBoundEvent}
   n7 = "Equal (Object)"(
    A=GetActiveWidget!(self:Widget Switcher Object Reference=Get PanelSwitcher() -> (PanelSwitcher:Widget Switcher Object Reference)) -> (ReturnValue:Widget Object Reference)
    B=Get Text_Overlay() -> (Text_Overlay:Widget Object Reference)
   )
   > ShowConfirmationDialog_NUI!(
    execute:Exec
    Icon:Slate Brush Structure
    Title:Text=Select(
     Option 0:Text=Get HelpStylesTabTitle() -> (HelpStylesTabTitle:Text)
     Option 1:Text=Get HelpTextTabTitle() -> (HelpTextTabTitle:Text)
     Index=@n7
    ) -> (ReturnValue:Text)
    Message:Text=Select(Option 0=Get HelpStylesTabDesc() -> (HelpStylesTabDesc:Text), Option 1=Get HelpTextTabDesc() -> (HelpTextTabDesc:Text), Index=@n7)
    ConfirmButtonInputActions:Array of Confirmation Dialog Action Structures=Make Array("[0]":Confirmation Dialog Action Structure=Make Confirmation Dialog Action!(
      DisplayName:Text="LOCTEXT(\"Got It\")"
      ResultName:Name=None
      Icon:Slate Brush Structure=(TintColor=(SpecifiedColor=(R=1,G=1,B=1,A=1),ColorUseRule=UseColor_Specified),DrawAs=Image,Tiling=NoTile,Mirroring=NoMirror,ImageType=NoImage,ImageSize=(X=32,Y=32),Margin=(Left=0,Top=0,Right=0,Bottom=0),ResourceObject=None,OutlineSettings=(CornerRadii=(X=0,Y=0,Z=0,W=1),Color=(SpecifiedColor=(R=0,G=0,B=0,A=0),ColorUseRule=UseColor_Specified),Width=0,RoundingType=HalfHeightRadius,bUseBrushTransparency=False),UVRegion=(Min=(X=0,Y=0),Max=(X=0,Y=0),bIsValid=False),bIsDynamicallyLoaded=False,ResourceName="")
      ActionName:Name=GenericBack
     ) -> (ConfirmationDialogAction:Confirmation Dialog Action Structure) {is=title K2Node_MakeStruct}) -> (Array:Array of Confirmation Dialog Action Structures)
    DeclineButtonInputAction:"Name (by ref)"=MakeLiteralName(Value=None)
    AdditionalContent:Widget Object Reference
    LeftAdditionalContent:Widget Object Reference
    TargetPlayer:Local Player Object Reference
   ) -> (then:Exec, DialogResult:Exec, Result:EFortDialogResult Enum, ResultName:Name) {is=title K2Node_AsyncAction}
   n8 = HandleInputMethodChanged!() -> (OutputDelegate, then, bNewInputType:ECommonInputType Enum) {is=title K2Node_CustomEvent}
   > Set bStylingInProgress(bStylingInProgress=false)
   > n9 = SetCommitCancelState(OkEnabled=true, CancelEnabled=true)
   Apply Button Visibility(execute=[@n9.then, @n26.then])
   n10 = Setup Rich Text Styles!() -> (OutputDelegate, then, Owning Device Object:Object Reference) {is=title K2Node_CustomEvent}
   > n11 = Get Device Rich Text Overrides!(execute, self, Object:Object Reference=@n10.Owning Device Object) -> (then, OverrideStyle:bool, RichTextStyle:Data Table Object Reference) {is=title K2Node_CallFunction}
   > n12 = Branch(Condition=@n11)
   > n13 = SetRichStyleSetOverride!(execute, self:Fort Creative Rich Text Editor Widget Object Reference, RichStyleSetOverride:Data Table Object Reference=@n11.RichTextStyle) -> (then)
   n14 = Get!(execute:Exec=[@n13.then, @n12.else]) -> (then:Exec, else:Exec, DefaultStyleButton:Style Button Object Reference)
   > UpdateDefaultStyle!(execute, self:Style Button Object Reference=@n14, Initiating Device:Object Reference=Get Device() -> (Device:Object Reference)) -> (then)
   n15 = Initialize Text!() -> (OutputDelegate, then, inText:Text) {is=title K2Node_CustomEvent}
   > SetTargetText!(execute, self, inText:Text=@n15.inText) -> (then)
   > Initialize!(execute, self, inTextToBeEdited:String=Get OriginalInputString() -> (OriginalInputString:String)) -> (then)
   Event Construct()
   > Set Style Editor State!(execute, self, Enabled:bool=CanShowStyleEditor!(self:Fort Creative Rich Text Editor Widget Object Reference) -> (ReturnValue:bool)) -> (then) {is=title K2Node_CallFunction}
   > SetActiveWidgetIndex(self=Get PanelSwitcher(), Index=0)
   > Apply Button Visibility()
   > SetCommitCancelState(OkEnabled=true, CancelEnabled=true)
   > SetIsInteractionEnabled!(execute, self:Common Button Base Object Reference=Get Button_WordButtonLabel(), bInIsInteractionEnabled:bool=false) -> (then)
   n16 = Event OnUpdateCommitCancelState!() -> (OutputDelegate, then, Enabled:bool) {is=title K2Node_Event}
   > SetCommitCancelState(OkEnabled=@n16.Enabled, CancelEnabled=@n16.Enabled)
   n17 = Event OnWordButtonCreated!() -> (OutputDelegate, then, WordButton:Fort Creative Rich Text Word Button Object Reference) {is=title K2Node_Event}
   > n18 = Cast To StyleEditorWordItem!(execute, Object=@n17.WordButton) -> (then, CastFailed, AsStyle Editor Word Item:Style Editor Word Item Object Reference) {is=title K2Node_DynamicCast}
   > n19 = Cast To BPi_TextStyleOverrides!(execute, Object=Get Device()) -> (then, CastFailed, AsBPi Text Style Overrides:BPi Text Style Overrides Interface) {is=title K2Node_DynamicCast}
   > UpdateDefaultStyle!(execute, self:Style Editor Word Item Object Reference=@n18, Target:BPi Text Style Overrides Interface=@n19) -> (then)
   n20 = Event BP_DialogConfirmed!() -> (OutputDelegate, then, DialogType:EConfirmDialogType Enum=CancelChanges) {is=title K2Node_Event}
   > n21 = Switch on EConfirmDialogType!(execute, Selection:EConfirmDialogType Enum=@n20.DialogType) -> (CancelChanges:Exec, CharLimitExceeded:Exec, MarkupWarning:Exec) {is=title K2Node_SwitchEnum}
   > n22 = Call On Text Change Confirmed!(execute:Exec, self:Text And Style Editor Object Reference, InString:String, ShouldUpdate:bool) -> (then:Exec) {is=title K2Node_CallDelegate}
   PopCurrentModal!(execute=[@n24.then, @n22.then]) -> (then)
   > DeactivateWidget()
   n23 = GetCurrentText!(execute=@n21.CharLimitExceeded, self) -> (then, CurrentText:String)
   > n24 = Call On Text Change Confirmed(InString=@n23, ShouldUpdate=true)
   n25 = Event BP_WordButtonHovered!() -> (OutputDelegate, then, Button:Fort Creative Rich Text Word Button Object Reference) {is=title K2Node_Event}
   > Style Editor Word Button Hovered!(execute, self, AssociatedButton:Common Button Base Object Reference=@n25.Button) -> (then)
   n26 = Event OnUpdateButtonsVisibility() {is=title K2Node_Event}
   Event OnFormatError!() -> (OutputDelegate, then, Description:"Text (by ref)") {is=title K2Node_Event}
   > ShowConfirmationDialog(DialogType=MarkupWarning)
   Event OnInitializedFinished() {is=title K2Node_Event}
   > Bind Event to On Input Method Changed(self=Get CommonInputSubsystem(), Delegate=@n8)
   n27 = Event PreConstruct()
   > Branch(Condition=@n27.IsDesignTime)
   > SetTargetText(inText="LOCTEXT(\"It\\'s dangerous to go alone. Take this! Also, take my cat. His name is tuna and he hates fish but he loves mice....\")")
   n28 = Event OnStyleButtonCreated!() -> (
    OutputDelegate
    then
    Button:Common Button Base Object Reference
    Label:String
    FormattedText:String
    TextStyleSet:Data Table Object Reference
    bIsDefaultButton:bool
   ) {is=title K2Node_Event}
   > n29 = Cast To StyleButton!(execute, Object=@n28.Button) -> (then, CastFailed, AsStyle Button:Style Button Object Reference) {is=title K2Node_DynamicCast}
   > SetupStyleButton!(
    execute
    self
    Button:Style Button Object Reference=@n29
    Label:String=@n28.Label
    SampleFormattedText:String=@n28.FormattedText
    StyleSet:Data Table Object Reference=@n28.TextStyleSet
   ) -> (then)
   > Branch(Condition=@n28.bIsDefaultButton)
   > Set DefaultStyleButton(DefaultStyleButton:Style Button Object Reference=@n29) -> (Output_Get:Style Button Object Reference)
   n30 = Event OnCharLimitReached!() -> (OutputDelegate, then, bIsReached:bool) {is=title K2Node_Event}
   > n31 = Branch(Condition=@n30.bIsReached)
   n32 = Get Ani_Warning() -> (Ani_Warning:Widget Animation Object Reference)
   > PlayAnimationForward(InAnimation=@n32, PlaybackSpeed=0.2)
   n33 = Get Border_CharWarning() -> (Border_CharWarning:Border Object Reference)
   > SetBrush(self:Border Object Reference=@n33, InBrush=Get Brush_CharsNotInRange() -> (Brush_CharsNotInRange:Slate Brush Structure))
   PlayAnimationReverse(execute=@n31.else, InAnimation=@n32)
   > SetBrush(self=@n33, InBrush=Get Brush_CharsInRange() -> (Brush_CharsInRange:Slate Brush Structure))
   n34 = Event OnStyleButtonSelectedChanged!() -> (OutputDelegate, then, Button:Common Button Base Object Reference, Selected:bool) {is=title K2Node_Event}
   > n35 = Cast To StyleButton(Object=@n34.Button)
   > SetSelected!(execute, self:Style Button Object Reference=@n35, Selected:bool=@n34.Selected) -> (then)
   n36 = Event OnUpdateWordButtonLabel!() -> (OutputDelegate, then, ButtonText:"Text (by ref)") {is=title K2Node_Event}
   > Set Text!(execute, self:Icon Text Button Object Reference=Get Button_WordButtonLabel(), ButtonText:Text=@n36.ButtonText) -> (then) {is=title K2Node_CallFunction}
   Event Tick()
  ]}
 Get Device Rich Text Overrides {nodes [
   n1 = Get Device Rich Text Overrides!() -> (then, Object:Object Reference) {is=title K2Node_FunctionEntry}
   > n2 = Set Device(Device:Object Reference=@n1) -> (Output_Get:Object Reference)
   > n3 = Cast To BPi_TextStyleOverrides(Object=@n2)
   > n4 = GetTextStyleOverride!(execute, self:BPi Text Style Overrides Interface=@n3) -> (then, HasOverride:bool, RichTextStyle:Data Table Object Reference)
   > Return Node!(execute, OverrideStyle:bool=@n4, RichTextStyle:Data Table Object Reference=@n4.RichTextStyle)
   Return Node(execute=@n3.CastFailed, OverrideStyle=false)
  ]}
 Get Widget Context Visibility {nodes [
   Get Widget Context Visibility() {is=title K2Node_FunctionEntry}
   n1 = GetCurrentInputType(self=Get CommonInputSubsystem())
   n2 = GetActiveWidgetIndex!(self:Widget Switcher Object Reference=Get PanelSwitcher()) -> (ReturnValue:int=0)
   > Return Node!(execute, IsVisible:bool=AND Boolean(A="Equal (Enum)"(A=@n1, B=Gamepad), B="Equal (Integer)"(A=@n2, B=1)))
  ]}
 GetCurrentText {nodes [
   GetCurrentText() {is=title K2Node_FunctionEntry}
   n1 = Get TabsButtonGroup()
   > n2 = Switch on Int!(execute:Exec, Selection:int=GetSelectedButtonIndex!(self:Common Button Group Base Object Reference=@n1) -> (ReturnValue:int=0)) -> (Default:Exec, 0:Exec, 1:Exec)
   n3 = GetTextInRange!(execute=[@n4.then, @n2.0], self:Fort Creative Rich Text Editor Widget Object Reference) -> (then, ReturnValue:String)
   > Return Node!(execute, CurrentText:String=@n3)
   n4 = SelectButtonAtIndex!(execute=@n2.1, self:Common Button Group Base Object Reference=@n1, ButtonIndex:int=0, bAllowSound:bool=true) -> (then)
  ]}
 HandleStyleButtonHovered {nodes [
   n1 = HandleStyleButtonHovered!() -> (then, Button:Common Button Base Object Reference) {is=title K2Node_FunctionEntry}
   > ScrollWidgetIntoView!(
    execute
    self:Scroll Box Object Reference=Get StyleButtons_Scrollbox() -> (StyleButtons_Scrollbox:Scroll Box Object Reference)
    WidgetToFind:Widget Object Reference=@n1
    AnimateScroll:bool=true
    ScrollDestination:EDescendantScrollDestination Enum=IntoView
    Padding:float=0
   ) -> (then)
  ]}
 Initialize {nodes [
   Initialize!() -> (then, inTextToBeEdited:String) {is=title K2Node_FunctionEntry}
   > SelectButtonAtIndex(self=Get TabsButtonGroup(), ButtonIndex=0, bAllowSound=true)
   n1 = Get EditableText()
   > "SetText (Multi-Line Editable Text)"(self=@n1, InText="To Text (String)"(InString=Get inTextToBeEdited() -> (inTextToBeEdited:String)))
   > SetFocus(self=@n1)
  ]}
 OnMouseButtonDown {nodes [
   n1 = OnMouseButtonDown!() -> (then, MyGeometry:Geometry Structure, MouseEvent:"Pointer Event Structure (by ref)") {is=title K2Node_FunctionEntry}
   n2 = GetVisibility!(self:Widget Object Reference=Get Button_ClearSelection()) -> (ReturnValue:ESlateVisibility Enum=Visible)
   n3 = GetSelectedButtonBase!(self:Common Button Group Base Object Reference=Get TabsButtonGroup()) -> (ReturnValue:Common Button Base Object Reference)
   n4 = Get Style_Tab_Button() -> (Style_Tab_Button:Common Button Base Object Reference)
   > n5 = Branch(Condition=AND Boolean!(
     A=Is Mouse Button Down!(Input:"Pointer Event Structure (by ref)"=@n1.MouseEvent, MouseButton:Key Structure=RightMouseButton) -> (ReturnValue:bool)
     B="Equal (Object)"(A=@n3, B=@n4)
     C:bool="Not Equal (Enum)"!(A:ESlateVisibility Enum=@n2, B:ESlateVisibility Enum=Collapsed) -> (ReturnValue:bool)
    ) -> (ReturnValue))
   > ClearSelectedWordButtons()
   > Apply Button Visibility()
   > n6 = Set Handled(Handled:bool=true) -> (Output_Get:bool)
   Return Node!(
    execute=[@n6.then, @n7.then]
    ReturnValue:Event Reply Structure=Select(Option 0:Event Reply Structure=Unhandled(), Option 1:Event Reply Structure=Handled(), Index=Get Handled() -> (Handled:bool)) -> (ReturnValue:Event Reply Structure)
   )
   n7 = Set Handled(execute=@n5.else, Handled=false)
  ]}
 OnTextChangeConfirmed {nodes [OnTextChangeConfirmed!() -> (then, InString:String, ShouldUpdate:bool=~) {is=title K2Node_FunctionEntry}]}
 Set Style Editor State {nodes [
   n1 = Set Style Editor State!() -> (then, Enabled:bool=~) {is=title K2Node_FunctionEntry}
   > SetVisibility(
    self=[
     Get HorizontalBox_TabButtons() -> (HorizontalBox_TabButtons:Horizontal Box Object Reference)
     Get SizeBox_ClearStyles() -> (SizeBox_ClearStyles:Size Box Object Reference)
    ]
    InVisibility=Select(Option 0:ESlateVisibility Enum=Collapsed, Option 1:ESlateVisibility Enum=SelfHitTestInvisible, Index=@n1) -> (ReturnValue:ESlateVisibility Enum)
   )
  ]}
 SetCommitCancelState {nodes [
   n1 = SetCommitCancelState!() -> (then, OkEnabled:bool=~, CancelEnabled:bool=~) {is=title K2Node_FunctionEntry}
   > SetIsEnabled(self=Get SaveAndClose_Button() -> (SaveAndClose_Button:WBP UIKit Button Regular Object Reference), bInIsEnabled=@n1)
   > SetIsEnabled(self=Get Button_Cancel() -> (Button_Cancel:Icon Text Button Object Reference), bInIsEnabled=@n1.CancelEnabled)
  ]}
 SetDeviceInfo {nodes [
   n1 = SetDeviceInfo!() -> (then, DeviceName:Text, DeviceFieldName:Text) {is=title K2Node_FunctionEntry}
   > n2 = Branch(Condition=TextIsEmpty(InText=@n1))
   "SetText (Text)"(
    execute=[@n3.then, @n2.then]
    self=Get Text_DeviceFieldName() -> (Text_DeviceFieldName:Common Text Object Reference)
    InText=@n1.DeviceFieldName
   )
   n3 = "SetText (Text)"(execute=@n2.else, self=Get Text_DeviceName() -> (Text_DeviceName:Common Text Object Reference), InText=@n1)
  ]}
 SetEditableTextFocus {nodes [
   SetEditableTextFocus() {is=title K2Node_FunctionEntry}
   > SetFocus(self=Get EditableText())
  ]}
 SetTargetText {nodes [
   n1 = SetTargetText!() -> (then, inText:Text) {is=title K2Node_FunctionEntry}
   > Set OriginalInputString(OriginalInputString:String="To String (Text)"(InText=@n1)) -> (Output_Get:String)
  ]}
 SetupStyleButton {nodes [
   n1 = SetupStyleButton!() -> (then, Button:Style Button Object Reference, Label:String, SampleFormattedText:String, StyleSet:Data Table Object Reference) {is=title K2Node_FunctionEntry}
   > SetupStyleButton!(
    execute
    self:Style Button Object Reference=@n1
    Label:Text="To Text (String)"(InString=@n1.Label)
    SampleText:Text="To Text (String)"(InString=@n1.SampleFormattedText)
    StyleSet=@n1.StyleSet
    InShowRichTextBox:bool=true
   ) -> (then) {is=title K2Node_CallFunction}
   > Bind Event to On Hovered(self:Common Button Base Object Reference=@n1, Delegate=Create Event()) {is=title K2Node_AddDelegate}
  ]}
 Style Editor Word Button Hovered {nodes [
   n1 = Style Editor Word Button Hovered!() -> (then, AssociatedButton:Common Button Base Object Reference) {is=title K2Node_FunctionEntry}
   > ScrollWidgetIntoView(
    self=Get ScrollBoxWords() -> (ScrollBoxWords:Scroll Box Object Reference)
    WidgetToFind=@n1
    AnimateScroll=true
    ScrollDestination=IntoView
    Padding=0
   )
  ]}
]
variables [
 Brush_CharsInRange {
  type=SlateBrush
  default {
   bIsDynamicallyLoaded=false
   drawAs=RoundedBox
   imageType=NoImage
   mirroring=NoMirror
   resourceName=None
   resourceObject=null
   tiling=NoTile
   imageSize=<32 32>
   margin=<ltrb 0 0 0 0>
   outlineSettings {
    bUseBrushTransparency=false
    roundingType=FixedRadius
    width=6
    color {colorUseRule=UseColor_Specified, specifiedColor=<rgba 1 1 1 1>}
    cornerRadii=<0 0 0 0>
   }
   tintColor {colorUseRule=UseColor_Specified, specifiedColor=<rgba 1 1 1 0>}
   uVRegion {bIsValid=false, max=<0 0>, min=<0 0>}
  }
 }
 Brush_CharsNotInRange {
  type=SlateBrush
  default {
   bIsDynamicallyLoaded=false
   drawAs=RoundedBox
   imageType=NoImage
   mirroring=NoMirror
   resourceName=None
   resourceObject=null
   tiling=NoTile
   imageSize=<32 32>
   margin=<ltrb 0 0 0 0>
   outlineSettings {
    bUseBrushTransparency=false
    roundingType=FixedRadius
    width=6
    color {colorUseRule=UseColor_Specified, specifiedColor=<rgba 0.7913 0.0086 0.1845 1>}
    cornerRadii=<0 0 0 0>
   }
   tintColor {colorUseRule=UseColor_Specified, specifiedColor=<rgba 1 1 1 0>}
   uVRegion {bIsValid=false, max=<0 0>, min=<0 0>}
  }
 }
 ButtonUpdateFailed {type=boolean}
 DefaultStyleButton {type=StyleButton_C, default=null}
 Device {type=Object, default=null}
 HelpStylesTabDesc {
  type=string
  default="STYLE TAB lets you emphasize sentences of your text, remark main words, call out on important information or even make your text more fun to read!!!\r\nPick as many words as you wish. When you are done with the selection, move to the list on the right.\r\nTry as many styles as you want. When you are sure of your choice, clear the selection, so you can style other group of words.\r\n\r\nStyles are defined by an opening <ITALIC> and closing </> markup, if you move to the TEXT tab, you'll see your text is full of them. They go in pairs, so do not erase any of them or your text won't show properly. "
 }
 HelpStylesTabTitle {type=string, default=Styling Text}
 HelpTextTabDesc {
  type=string
  default="On this section you can write the message you want to show your players.\r\nUse {playername} to dynamically include the player name on the text.\r\nPress \"Clear Text\" button to erase all the content.\r\nBe aware of the character limit, don't close the widget window before fixing any errors.\r\nIf you applied style to the text, you'll notice the opening <ITALIC> and closing </> markups. Do not delete any of them or your text won't show properly. You can clear all the styled text with the \"Clear Styles\" button."
 }
 HelpTextTabTitle {type=string, default=Edit Your Text}
 OnTextChangeConfirmed {}
 OriginalInputString {type=string, default=""}
]
widgets {
 info {
  inheritedWidgetCount=11
  parentClass=/Script/FortniteUI.FortCreativeRichTextEditorWidget
  rootWidgetClass=&Overlay
  widgetCount=76
 }
 widgets [
  Overlay_SelectTags {widgetClassPath=&Overlay, bIsVariable=true, widget=Overlay_SelectTags}
  Lightbox {
   widgetClassPath=/Game/UI/Foundation/Lightbox/Lightbox
   bIsVariable=true
   parent=Overlay_SelectTags
   widget=Lightbox
   properties {dimAmount=0.9, showWindowControlButtons=false}
   slot {class=OverlaySlot, properties {horizontalAlignment=HAlign_Fill, verticalAlignment=VAlign_Fill}}
  }
  SizeBox_0 {
   widgetClassPath=&SizeBox
   parent=Overlay_SelectTags
   widget=SizeBox_0
   properties {bOverride_HeightOverride=true, bOverride_WidthOverride=true, heightOverride=810, widthOverride=1920}
   slot {class=OverlaySlot, properties {horizontalAlignment=HAlign_Fill, verticalAlignment=VAlign_Fill, padding=<ltrb 0 104 0 167>}}
  }
  Overlay_708 {widgetClassPath=&Overlay, parent=SizeBox_0, widget=Overlay_708, slot {class=SizeBoxSlot}}
  Image_1 {
   widgetClassPath=&Image
   parent=Overlay_708
   widget=Image_1
   properties {brush {
     bIsDynamicallyLoaded=false
     drawAs=Image
     imageType=NoImage
     mirroring=NoMirror
     resourceName=None
     resourceObject=null
     tiling=NoTile
     imageSize=<32 32>
     margin=<ltrb 0 0 0 0>
     outlineSettings {
      bUseBrushTransparency=false
      roundingType=HalfHeightRadius
      width=0
      color {colorUseRule=UseColor_Specified, specifiedColor=<rgba 0 0 0 0>}
      cornerRadii=<0 0 0 0>
     }
     tintColor {colorUseRule=UseColor_Specified, specifiedColor=<rgba 0.0024 0.0262 0.2346 1>}
     uVRegion {bIsValid=false, max=<0 0>, min=<0 0>}
    }}
   slot {class=OverlaySlot, properties {horizontalAlignment=HAlign_Fill, verticalAlignment=VAlign_Fill}}
  }
  Image_BackgroundBottomBar {
   widgetClassPath=&Image
   parent=Overlay_708
   widget=Image_BackgroundBottomBar
   properties {brush {
     bIsDynamicallyLoaded=false
     drawAs=Image
     imageType=NoImage
     mirroring=NoMirror
     resourceName=None
     resourceObject=/FortUILibrary/Foundation/Art_ToolKit/AngledBoxes/M_UI_MoreInfo_BottomDividerLIne_Blue
     tiling=NoTile
     imageSize=<32 28>
     margin=<ltrb 0 0 0 0>
     outlineSettings {
      bUseBrushTransparency=false
      roundingType=HalfHeightRadius
      width=0
      color {colorUseRule=UseColor_Specified, specifiedColor=<rgba 0 0 0 0>}
      cornerRadii=<0 0 0 0>
     }
     tintColor {colorUseRule=UseColor_Specified, specifiedColor=<rgba 0.2462 0.2747 0.4793 1>}
     uVRegion {bIsValid=false, max=<0 0>, min=<0 0>}
    }}
   slot {class=OverlaySlot, properties {horizontalAlignment=HAlign_Fill, verticalAlignment=VAlign_Bottom}}
  }
  VerticalBox_2 {
   widgetClassPath=&VerticalBox
   parent=Overlay_708
   widget=VerticalBox_2
   slot {class=OverlaySlot, properties {horizontalAlignment=HAlign_Fill, verticalAlignment=VAlign_Fill}}
  }
  Overlay_Title {
   widgetClassPath=&Overlay
   parent=VerticalBox_2
   widget=Overlay_Title
   slot {class=VerticalBoxSlot, properties {verticalAlignment=VAlign_Top}}
  }
  Spacer_94 {
   widgetClassPath=&Spacer
   parent=Overlay_Title
   widget=Spacer_94
   properties {size=<1 160>}
   slot {class=OverlaySlot, properties {horizontalAlignment=HAlign_Fill, verticalAlignment=VAlign_Fill}}
  }
  Title_VBox {
   widgetClassPath=&VerticalBox
   parent=Overlay_Title
   widget=Title_VBox
   slot {class=OverlaySlot, properties {horizontalAlignment=HAlign_Fill, verticalAlignment=VAlign_Fill}}
  }
  Title {
   widgetClassPath=&SizeBox
   parent=Title_VBox
   widget=Title
   properties {bOverride_HeightOverride=true, heightOverride=160}
   slot {class=VerticalBoxSlot}
  }
  Border_54 {
   widgetClassPath=&Border
   parent=Title
   widget=Border_54
   properties {
    horizontalAlignment=HAlign_Left
    verticalAlignment=VAlign_Center
    background {
     bIsDynamicallyLoaded=false
     drawAs=Image
     imageType=NoImage
     mirroring=NoMirror
     resourceName=None
     resourceObject=null
     tiling=NoTile
     imageSize=<32 32>
     margin=<ltrb 0 0 0 0>
     outlineSettings {
      bUseBrushTransparency=false
      roundingType=HalfHeightRadius
      width=0
      color {colorUseRule=UseColor_Specified, specifiedColor=<rgba 0 0 0 0>}
      cornerRadii=<0 0 0 0>
     }
     tintColor {colorUseRule=UseColor_Specified, specifiedColor=<rgba 0.0024 0.016 0.117 1>}
     uVRegion {bIsValid=false, max=<0 0>, min=<0 0>}
    }
    brushColor=<rgba 1 1 1 0.7>
    padding=<ltrb 252 0 0 0>
   }
   slot {class=SizeBoxSlot}
  }
  VerticalBox_148 {
   widgetClassPath=&VerticalBox
   parent=Border_54
   widget=VerticalBox_148
   slot {class=BorderSlot, properties {horizontalAlignment=HAlign_Left, verticalAlignment=VAlign_Center, padding=<ltrb 252 0 0 0>}}
  }
  Text_DeviceName {
   widgetClassPath=&CommonTextBlock
   bIsVariable=true
   parent=VerticalBox_148
   widget=Text_DeviceName
   properties {
    renderOpacity=0.5
    style=/FortUILibrary/Styles/Typography/_TextStyle_SubH4._TextStyle_SubH4_C
    text=POP-UP DIALOG DEVICE
    textTransformPolicy=ToUpper
    font {
     bForceMonospaced=false
     bMaterialIsStencil=false
     fontMaterial=null
     fontObject=&BurbankSmall
     letterSpacing=0
     monospacedWidth=1
     size=24
     skewAmount=0
     typefaceFontName=Black
     outlineSettings {
      bApplyOutlineToDropShadows=false
      bMiteredCorners=false
      bSeparateFillAlpha=false
      outlineMaterial=null
      outlineSize=0
      outlineColor=<rgba 0 0 0 1>
     }
    }
    shadowOffset=<0 0>
    strikeBrush {
     bIsDynamicallyLoaded=false
     drawAs=NoDrawType
     imageType=NoImage
     mirroring=NoMirror
     resourceName=None
     resourceObject=null
     tiling=NoTile
     imageSize=<32 32>
     margin=<ltrb 0 0 0 0>
     outlineSettings {
      bUseBrushTransparency=false
      roundingType=HalfHeightRadius
      width=0
      color {colorUseRule=UseColor_Specified, specifiedColor=<rgba 0 0 0 0>}
      cornerRadii=<0 0 0 0>
     }
     tintColor {colorUseRule=UseColor_Specified, specifiedColor=<rgba 1 1 1 1>}
     uVRegion {bIsValid=false, max=<0 0>, min=<0 0>}
    }
   }
   slot {class=VerticalBoxSlot, properties {horizontalAlignment=HAlign_Left}}
  }
  Text_DeviceFieldName {
   widgetClassPath=&CommonTextBlock
   bIsVariable=true
   parent=VerticalBox_148
   widget=Text_DeviceFieldName
   properties {
    style=/FortUILibrary/Styles/Typography/_TextStyle_PrimaryH2._TextStyle_PrimaryH2_C
    text=DESCRIPTION TEXT
    textTransformPolicy=ToUpper
    font {
     bForceMonospaced=false
     bMaterialIsStencil=false
     fontMaterial=null
     fontObject=&Burbank
     letterSpacing=0
     monospacedWidth=1
     size=48
     skewAmount=0
     typefaceFontName=BurbankRegular-Black
     outlineSettings {
      bApplyOutlineToDropShadows=false
      bMiteredCorners=false
      bSeparateFillAlpha=false
      outlineMaterial=null
      outlineSize=0
      outlineColor=<rgba 0 0 0 1>
     }
    }
    shadowOffset=<0 0>
    strikeBrush {
     bIsDynamicallyLoaded=false
     drawAs=NoDrawType
     imageType=NoImage
     mirroring=NoMirror
     resourceName=None
     resourceObject=null
     tiling=NoTile
     imageSize=<32 32>
     margin=<ltrb 0 0 0 0>
     outlineSettings {
      bUseBrushTransparency=false
      roundingType=HalfHeightRadius
      width=0
      color {colorUseRule=UseColor_Specified, specifiedColor=<rgba 0 0 0 0>}
      cornerRadii=<0 0 0 0>
     }
     tintColor {colorUseRule=UseColor_Specified, specifiedColor=<rgba 1 1 1 1>}
     uVRegion {bIsValid=false, max=<0 0>, min=<0 0>}
    }
   }
   slot {class=VerticalBoxSlot, properties {horizontalAlignment=HAlign_Left, verticalAlignment=VAlign_Top}}
  }
  Image_BackgroundBottomBar_1 {
   widgetClassPath=&Image
   parent=Title_VBox
   widget=Image_BackgroundBottomBar_1
   properties {brush {
     bIsDynamicallyLoaded=false
     drawAs=Image
     imageType=NoImage
     mirroring=NoMirror
     resourceName=None
     resourceObject=&M_UI_MoreInfo_BottomDividerLIne_Blue
     tiling=NoTile
     imageSize=<32 20>
     margin=<ltrb 0 0 0 0>
     outlineSettings {
      bUseBrushTransparency=false
      roundingType=HalfHeightRadius
      width=0
      color {colorUseRule=UseColor_Specified, specifiedColor=<rgba 0 0 0 0>}
      cornerRadii=<0 0 0 0>
     }
     tintColor {colorUseRule=UseColor_Specified, specifiedColor=<rgba 0.2462 0.2747 0.4793 1>}
     uVRegion {bIsValid=false, max=<0 0>, min=<0 0>}
    }}
   slot {class=VerticalBoxSlot, properties {verticalAlignment=VAlign_Top, padding=<ltrb 0 -16 0 0>}}
  }
  Overlay_TabButtons {
   widgetClassPath=&Overlay
   bIsVariable=true
   parent=VerticalBox_2
   widget=Overlay_TabButtons
   slot {class=VerticalBoxSlot, properties {padding=<ltrb 0 6 0 0>}}
  }
  Spacer_182 {widgetClassPath=&Spacer, parent=Overlay_TabButtons, widget=Spacer_182, properties {size=<1 112>}, slot {class=OverlaySlot}}
  HorizontalBox_TabButtons {
   widgetClassPath=&HorizontalBox
   bIsVariable=true
   parent=Overlay_TabButtons
   widget=HorizontalBox_TabButtons
   slot {class=OverlaySlot, properties {horizontalAlignment=HAlign_Fill, padding=<ltrb 252 4 0 0>}}
  }
  UserActionMenuInputButton_Left {
   widgetClassPath=/Game/UI/SocialPanel/UserActionMenuInputButton
   bIsVariable=true
   parent=HorizontalBox_TabButtons
   widget=UserActionMenuInputButton_Left
   properties {
    onButtonBaseClicked=(&self:WidgetTree.UserActionMenuInputButton_Left.HandleOnButtonClicked)
    onButtonBaseDoubleClicked=(&self:WidgetTree.UserActionMenuInputButton_Left.HandleOnButtonDoubleClicked)
    onButtonBaseHovered=(&self:WidgetTree.UserActionMenuInputButton_Left.HandleOnButtonHovered)
    onButtonBaseUnhovered=(&self:WidgetTree.UserActionMenuInputButton_Left.HandleOnButtonUnhovered)
    onSelectedChangedBase=(&self:WidgetTree.UserActionMenuInputButton_Left.HandleOnSelectedChanged)
    triggeringInputAction {dataTable=&InputActionDataTable, rowName=GenericLeftShoulder}
   }
   slot {class=HorizontalBoxSlot}
  }
  Text_Tab_Button {
   widgetClassPath=&IconTextButton
   bInherited=true
   bIsVariable=true
   parent=HorizontalBox_TabButtons
   widget=Text_Tab_Button
   properties {
    bHideInputAction=true
    bIsFocusable=false
    bSelectable=true
    buttonText=Edit Text
    contentAlignment=HAlign_Center
    minHeight=64
    minWidth=220
    onButtonBaseClicked=(&self:WidgetTree.Text_Tab_Button.HandleOnButtonClicked)
    onButtonBaseDoubleClicked=(&self:WidgetTree.Text_Tab_Button.HandleOnButtonDoubleClicked)
    onButtonBaseHovered=(&self:WidgetTree.Text_Tab_Button.HandleOnButtonHovered)
    onButtonBaseUnhovered=(&self:WidgetTree.Text_Tab_Button.HandleOnButtonUnhovered)
    onSelectedChangedBase=(&self:WidgetTree.Text_Tab_Button.HandleOnSelectedChanged)
    style=/FortUILibrary/Foundation/Buttons/ButtonStyles/Tab/ButtonStyle-Tab-Secondary.ButtonStyle-Tab-Secondary_C
    visibility=Visible
    triggeringInputAction {dataTable=&InputActionDataTable, rowName=GenericLeftShoulder}
   }
   slot {class=HorizontalBoxSlot}
  }
  Style_Tab_Button {
   widgetClassPath=&IconTextButton
   bInherited=true
   bIsVariable=true
   parent=HorizontalBox_TabButtons
   widget=Style_Tab_Button
   properties {
    bHideInputAction=true
    bIsFocusable=false
    bSelectable=true
    buttonText=Format Styles
    contentAlignment=HAlign_Center
    minHeight=64
    minWidth=220
    onButtonBaseClicked=(&self:WidgetTree.Style_Tab_Button.HandleOnButtonClicked)
    onButtonBaseDoubleClicked=(&self:WidgetTree.Style_Tab_Button.HandleOnButtonDoubleClicked)
    onButtonBaseHovered=(&self:WidgetTree.Style_Tab_Button.HandleOnButtonHovered)
    onButtonBaseUnhovered=(&self:WidgetTree.Style_Tab_Button.HandleOnButtonUnhovered)
    onSelectedChangedBase=(&self:WidgetTree.Style_Tab_Button.HandleOnSelectedChanged)
    style=&ButtonStyle-Tab-Secondary_C
    visibility=Visible
    triggeringInputAction {dataTable=&InputActionDataTable, rowName=GenericRightShoulder}
   }
   slot {class=HorizontalBoxSlot, properties {padding=<ltrb 6 0 0 0>}}
  }
  UserActionMenuInputButton_Right {
   widgetClassPath=&UserActionMenuInputButton
   bIsVariable=true
   parent=HorizontalBox_TabButtons
   widget=UserActionMenuInputButton_Right
   properties {
    onButtonBaseClicked=(&self:WidgetTree.UserActionMenuInputButton_Right.HandleOnButtonClicked)
    onButtonBaseDoubleClicked=(&self:WidgetTree.UserActionMenuInputButton_Right.HandleOnButtonDoubleClicked)
    onButtonBaseHovered=(&self:WidgetTree.UserActionMenuInputButton_Right.HandleOnButtonHovered)
    onButtonBaseUnhovered=(&self:WidgetTree.UserActionMenuInputButton_Right.HandleOnButtonUnhovered)
    onSelectedChangedBase=(&self:WidgetTree.UserActionMenuInputButton_Right.HandleOnSelectedChanged)
    triggeringInputAction {dataTable=&InputActionDataTable, rowName=GenericRightShoulder}
   }
   slot {class=HorizontalBoxSlot, properties {padding=<ltrb 6 0 0 0>}}
  }
  PanelSwitcher {
   widgetClassPath=&WidgetSwitcher
   bInherited=true
   bIsVariable=true
   parent=VerticalBox_2
   widget=PanelSwitcher
   properties {activeWidgetIndex=1}
   slot {class=VerticalBoxSlot, properties {padding=<ltrb 0 -9 0 0>}}
  }
  Text_Overlay {
   widgetClassPath=&Overlay
   bInherited=true
   bIsVariable=true
   parent=PanelSwitcher
   widget=Text_Overlay
   slot {class=WidgetSwitcherSlot, properties {verticalAlignment=VAlign_Top}}
  }
  TextTab_Overlay {
   widgetClassPath=&HorizontalBox
   bIsVariable=true
   parent=Text_Overlay
   widget=TextTab_Overlay
   slot {class=OverlaySlot, properties {horizontalAlignment=HAlign_Fill}}
  }
  Spacer_133 {
   widgetClassPath=&Spacer
   parent=TextTab_Overlay
   widget=Spacer_133
   properties {size=<252 1>}
   slot {class=HorizontalBoxSlot, properties {verticalAlignment=VAlign_Top, size {sizeRule=Automatic, value=0.2}}}
  }
  Overlay_PlainText {widgetClassPath=&Overlay, bIsVariable=true, parent=TextTab_Overlay, widget=Overlay_PlainText, slot {class=HorizontalBoxSlot}}
  SizeBox_1 {
   widgetClassPath=&SizeBox
   parent=Overlay_PlainText
   widget=SizeBox_1
   properties {bOverride_HeightOverride=true, bOverride_WidthOverride=true, heightOverride=440, widthOverride=696}
   slot {class=OverlaySlot}
  }
  EditText_Image {
   widgetClassPath=&Image
   parent=SizeBox_1
   widget=EditText_Image
   properties {
    brush {
     bIsDynamicallyLoaded=false
     drawAs=Box
     imageType=NoImage
     mirroring=NoMirror
     resourceName=None
     resourceObject=null
     tiling=NoTile
     imageSize=<32 32>
     margin=<ltrb 0 0 0 0>
     outlineSettings {
      bUseBrushTransparency=false
      roundingType=FixedRadius
      width=4
      color {colorUseRule=UseColor_Specified, specifiedColor=<rgba 0 0.5457 1 1>}
      cornerRadii=<20 20 20 20>
     }
     tintColor {colorUseRule=UseColor_Specified, specifiedColor=<rgba 1 1 1 1>}
     uVRegion {bIsValid=false, max=<0 0>, min=<0 0>}
    }
    colorAndOpacity=<rgba 0.0009 0.006 0.0395 1>
   }
   slot {class=SizeBoxSlot}
  }
  SizeBox_3 {
   widgetClassPath=&SizeBox
   parent=Overlay_PlainText
   widget=SizeBox_3
   properties {bOverride_HeightOverride=true, bOverride_WidthOverride=true, heightOverride=440, widthOverride=696}
   slot {class=OverlaySlot}
  }
  Border_CharWarning {
   widgetClassPath=&Border
   bIsVariable=true
   parent=SizeBox_3
   widget=Border_CharWarning
   properties {
    background {
     bIsDynamicallyLoaded=false
     drawAs=RoundedBox
     imageType=NoImage
     mirroring=NoMirror
     resourceName=None
     resourceObject=null
     tiling=NoTile
     imageSize=<32 32>
     margin=<ltrb 0 0 0 0>
     outlineSettings {
      bUseBrushTransparency=false
      roundingType=FixedRadius
      width=6
      color {colorUseRule=UseColor_Specified, specifiedColor=<rgba 1 1 1 1>}
      cornerRadii=<0 0 0 0>
     }
     tintColor {colorUseRule=UseColor_Specified, specifiedColor=<rgba 1 1 1 0>}
     uVRegion {bIsValid=false, max=<0 0>, min=<0 0>}
    }
    padding=<ltrb 0 0 0 0>
   }
   slot {class=SizeBoxSlot}
  }
  SizeBox {
   widgetClassPath=&SizeBox
   parent=Overlay_PlainText
   widget=SizeBox
   properties {bOverride_HeightOverride=true, bOverride_WidthOverride=true, heightOverride=360, widthOverride=650}
   slot {class=OverlaySlot, properties {horizontalAlignment=HAlign_Fill, verticalAlignment=VAlign_Fill}}
  }
  ScrollBox_272 {
   widgetClassPath=&ScrollBox
   parent=SizeBox
   widget=ScrollBox_272
   properties {
    allowOverscroll=false
    bAnimateWheelScrolling=true
    navigation=&self:WidgetTree.ScrollBox_272.WidgetNavigation_0
    scrollWhenFocusChanges=AnimatedScroll
   }
   slot {class=SizeBoxSlot, properties {padding=<ltrb 20 20 20 55>}}
  }
  EditableText {
   widgetClassPath=/Script/UMG.MultiLineEditableText
   bInherited=true
   bIsVariable=true
   parent=ScrollBox_272
   widget=EditableText
   properties {
    clearKeyboardFocusOnCommit=false
    hintText=Enter text here
    lineHeightPercentage=1.4
    navigation=&self:WidgetTree.EditableText.WidgetNavigation_0
    wrappingPolicy=AllowPerCharacterWrapping
    widgetStyle {
     overflowPolicy=Clip
     transformPolicy=None
     colorAndOpacity {colorUseRule=UseColor_Specified, specifiedColor=<rgba 1 1 1 1>}
     font {
      bForceMonospaced=false
      bMaterialIsStencil=false
      fontMaterial=null
      fontObject=&Roboto
      letterSpacing=0
      monospacedWidth=1
      size=20
      skewAmount=0
      typefaceFontName=Bold
      outlineSettings {
       bApplyOutlineToDropShadows=false
       bMiteredCorners=false
       bSeparateFillAlpha=false
       outlineMaterial=null
       outlineSize=0
       outlineColor=<rgba 0 0 0 1>
      }
     }
     highlightColor {colorUseRule=UseColor_Specified, specifiedColor=<rgba 0.02 0.3 0 1>}
     highlightShape {
      bIsDynamicallyLoaded=false
      drawAs=Box
      imageType=FullColor
      mirroring=NoMirror
      resourceName=../../../Engine/Content/Slate/Common/TextBlockHighlightShape.png
      resourceObject=null
      tiling=NoTile
      imageSize=<0 0>
      margin=<ltrb 0.375 0.375 0.375 0.375>
      outlineSettings {
       bUseBrushTransparency=false
       roundingType=HalfHeightRadius
       width=0
       color {colorUseRule=UseColor_Specified, specifiedColor=<rgba 0 0 0 0>}
       cornerRadii=<0 0 0 0>
      }
      tintColor {colorUseRule=UseColor_Specified, specifiedColor=<rgba 1 1 1 1>}
      uVRegion {bIsValid=false, max=<0 0>, min=<0 0>}
     }
     selectedBackgroundColor {colorUseRule=UseColor_Foreground, specifiedColor=<rgba 1 0 1 1>}
     shadowColorAndOpacity=<rgba 0 0 0 1>
     shadowOffset=<0 0>
     strikeBrush {
      bIsDynamicallyLoaded=false
      drawAs=Image
      imageType=NoImage
      mirroring=NoMirror
      resourceName=None
      resourceObject=null
      tiling=NoTile
      imageSize=<32 32>
      margin=<ltrb 0 0 0 0>
      outlineSettings {
       bUseBrushTransparency=false
       roundingType=HalfHeightRadius
       width=0
       color {colorUseRule=UseColor_Specified, specifiedColor=<rgba 0 0 0 0>}
       cornerRadii=<0 0 0 0>
      }
      tintColor {colorUseRule=UseColor_Specified, specifiedColor=<rgba 1 1 1 1>}
      uVRegion {bIsValid=false, max=<0 0>, min=<0 0>}
     }
     underlineBrush {
      bIsDynamicallyLoaded=false
      drawAs=Image
      imageType=NoImage
      mirroring=NoMirror
      resourceName=None
      resourceObject=null
      tiling=NoTile
      imageSize=<32 32>
      margin=<ltrb 0 0 0 0>
      outlineSettings {
       bUseBrushTransparency=false
       roundingType=HalfHeightRadius
       width=0
       color {colorUseRule=UseColor_Specified, specifiedColor=<rgba 0 0 0 0>}
       cornerRadii=<0 0 0 0>
      }
      tintColor {colorUseRule=UseColor_Specified, specifiedColor=<rgba 1 1 1 1>}
      uVRegion {bIsValid=false, max=<0 0>, min=<0 0>}
     }
    }
   }
   slot {class=ScrollBoxSlot}
  }
  Overlay_0 {
   widgetClassPath=&Overlay
   parent=Overlay_PlainText
   widget=Overlay_0
   slot {class=OverlaySlot, properties {horizontalAlignment=HAlign_Right, verticalAlignment=VAlign_Bottom, padding=<ltrb 0 0 20 20>}}
  }
  HorizontalBox_112 {
   widgetClassPath=&HorizontalBox
   parent=Overlay_0
   widget=HorizontalBox_112
   properties {visibility=HitTestInvisible}
   slot {class=OverlaySlot}
  }
  WarningImage {
   widgetClassPath=&Overlay
   bIsVariable=true
   parent=HorizontalBox_112
   widget=WarningImage
   properties {visibility=Hidden}
   slot {class=HorizontalBoxSlot, properties {verticalAlignment=VAlign_Bottom, padding=<ltrb 0 0 6 0>}}
  }
  Image_119 {
   widgetClassPath=&Image
   bIsVariable=true
   parent=WarningImage
   widget=Image_119
   properties {
    brush {
     bIsDynamicallyLoaded=false
     drawAs=Image
     imageType=NoImage
     mirroring=NoMirror
     resourceName=None
     resourceObject=/Game/UI/Foundation/Shell/Textures/T-Icon-Warning-32
     tiling=NoTile
     imageSize=<32 32>
     margin=<ltrb 0 0 0 0>
     outlineSettings {
      bUseBrushTransparency=false
      roundingType=HalfHeightRadius
      width=0
      color {colorUseRule=UseColor_Specified, specifiedColor=<rgba 0 0 0 0>}
      cornerRadii=<0 0 0 0>
     }
     tintColor {colorUseRule=UseColor_Specified, specifiedColor=<rgba 1 1 1 1>}
     uVRegion {bIsValid=false, max=<0 0>, min=<0 0>}
    }
    colorAndOpacity=<rgba 0.7913 0.0086 0.1845 1>
   }
   slot {class=OverlaySlot, properties {horizontalAlignment=HAlign_Center, verticalAlignment=VAlign_Center}}
  }
  Text_CharCount {
   widgetClassPath=&CommonTextBlock
   bInherited=true
   parent=HorizontalBox_112
   widget=Text_CharCount
   properties {
    mobileFontSizeMultiplier=1.5
    text=0 / 250
    colorAndOpacity {colorUseRule=UseColor_Specified, specifiedColor=<rgba 0.0296 1 0.2232 1>}
    font {
     bForceMonospaced=false
     bMaterialIsStencil=false
     fontMaterial=null
     fontObject=&BurbankSmall
     letterSpacing=0
     monospacedWidth=1
     size=16
     skewAmount=0
     typefaceFontName=Bold
     outlineSettings {
      bApplyOutlineToDropShadows=false
      bMiteredCorners=false
      bSeparateFillAlpha=false
      outlineMaterial=null
      outlineSize=1
      outlineColor=<rgba 0 0 0 0>
     }
    }
    shadowColorAndOpacity=<rgba 0 0 0 0.3>
    shadowOffset=<0.5 0.5>
   }
   slot {class=HorizontalBoxSlot, properties {horizontalAlignment=HAlign_Right, verticalAlignment=VAlign_Bottom}}
  }
  Spacer_250 {
   widgetClassPath=&Spacer
   parent=TextTab_Overlay
   widget=Spacer_250
   properties {size=<42 1>}
   slot {class=HorizontalBoxSlot, properties {size {sizeRule=Automatic, value=0.05}}}
  }
  VerticalBox_ActionButtons {
   widgetClassPath=&VerticalBox
   bIsVariable=true
   parent=TextTab_Overlay
   widget=VerticalBox_ActionButtons
   properties {navigation=&self:WidgetTree.VerticalBox_ActionButtons.WidgetNavigation_0}
   slot {class=HorizontalBoxSlot, properties {size {sizeRule=Fill, value=1}}}
  }
  SizeBox_ClearText {
   widgetClassPath=&SizeBox
   parent=VerticalBox_ActionButtons
   widget=SizeBox_ClearText
   properties {bOverride_HeightOverride=true, bOverride_WidthOverride=true, heightOverride=56, widthOverride=724}
   slot {class=VerticalBoxSlot, properties {horizontalAlignment=HAlign_Left}}
  }
  Border_3 {
   widgetClassPath=&Border
   parent=SizeBox_ClearText
   widget=Border_3
   properties {brushColor=<rgba 0.0024 0.0513 0.3864 1>, padding=<ltrb 0 0 0 0>, renderTransformPivot=<0 0>}
   slot {class=SizeBoxSlot}
  }
  ClearText_Button {
   widgetClassPath=&/StyleButton
   bIsVariable=true
   parent=Border_3
   widget=ClearText_Button
   properties {
    buttonText=Clear Text
    minWidth=600
    navigation=&self:WidgetTree.ClearText_Button.WidgetNavigation_0
    showRichTextBox=false
    iconBrush {
     bIsDynamicallyLoaded=false
     drawAs=Image
     imageType=NoImage
     mirroring=NoMirror
     resourceName=None
     resourceObject=/Game/Athena/UI/InGame/Creative/PropertyEditor/Icon_Picker/Textures_V2/64x64/T_UI_IconLibrary_TrashCan_64
     tiling=NoTile
     imageSize=<32 32>
     margin=<ltrb 0 0 0 0>
     outlineSettings {
      bUseBrushTransparency=false
      roundingType=HalfHeightRadius
      width=0
      color {colorUseRule=UseColor_Specified, specifiedColor=<rgba 0 0 0 0>}
      cornerRadii=<0 0 0 1>
     }
     tintColor {colorUseRule=UseColor_Specified, specifiedColor=<rgba 1 1 1 1>}
     uVRegion {bIsValid=false, max=<0 0>, min=<0 0>}
    }
   }
   slot {class=BorderSlot, properties {padding=<ltrb 0 0 0 0>}}
  }
  Spacer {widgetClassPath=&Spacer, parent=VerticalBox_ActionButtons, widget=Spacer, properties {size=<1 7>}, slot {class=VerticalBoxSlot}}
  SizeBox_ClearStyles {
   widgetClassPath=&SizeBox
   bIsVariable=true
   parent=VerticalBox_ActionButtons
   widget=SizeBox_ClearStyles
   properties {bOverride_HeightOverride=true, bOverride_WidthOverride=true, heightOverride=56, widthOverride=724}
   slot {class=VerticalBoxSlot, properties {horizontalAlignment=HAlign_Left}}
  }
  Border_4 {
   widgetClassPath=&Border
   parent=SizeBox_ClearStyles
   widget=Border_4
   properties {brushColor=<rgba 0.0024 0.0513 0.3864 1>, padding=<ltrb 0 0 0 0>}
   slot {class=SizeBoxSlot}
  }
  RemoveStyles_Button {
   widgetClassPath=&StyleButton
   bInherited=true
   bIsVariable=true
   parent=Border_4
   widget=RemoveStyles_Button
   properties {
    buttonText=Clear Styles
    minWidth=600
    navigation=&self:WidgetTree.RemoveStyles_Button.WidgetNavigation_0
    showRichTextBox=false
    iconBrush {
     bIsDynamicallyLoaded=false
     drawAs=Image
     imageType=NoImage
     mirroring=NoMirror
     resourceName=None
     resourceObject=&/T_UI_Icon_Clear_Style
     tiling=NoTile
     imageSize=<32 32>
     margin=<ltrb 0 0 0 0>
     outlineSettings {
      bUseBrushTransparency=false
      roundingType=HalfHeightRadius
      width=0
      color {colorUseRule=UseColor_Specified, specifiedColor=<rgba 0 0 0 0>}
      cornerRadii=<0 0 0 1>
     }
     tintColor {colorUseRule=UseColor_Specified, specifiedColor=<rgba 1 1 1 1>}
     uVRegion {bIsValid=false, max=<0 0>, min=<0 0>}
    }
   }
   slot {class=BorderSlot, properties {padding=<ltrb 0 0 0 0>}}
  }
  Spacer_3 {widgetClassPath=&Spacer, parent=VerticalBox_ActionButtons, widget=Spacer_3, properties {size=<1 7>}, slot {class=VerticalBoxSlot}}
  SizeBox_EditText {
   widgetClassPath=&SizeBox
   parent=VerticalBox_ActionButtons
   widget=SizeBox_EditText
   properties {bOverride_HeightOverride=true, bOverride_WidthOverride=true, heightOverride=56, visibility=Hidden, widthOverride=724}
   slot {class=VerticalBoxSlot, properties {horizontalAlignment=HAlign_Left, verticalAlignment=VAlign_Center}}
  }
  Border_0 {
   widgetClassPath=&Border
   parent=SizeBox_EditText
   widget=Border_0
   properties {brushColor=<rgba 0.0024 0.0513 0.3864 1>, padding=<ltrb 0 0 0 0>}
   slot {class=SizeBoxSlot}
  }
  EditText_Button {
   widgetClassPath=&StyleButton
   bIsVariable=true
   parent=Border_0
   widget=EditText_Button
   properties {
    bIsEnabled=false
    buttonText=Edit Text
    minWidth=600
    showRichTextBox=false
    visibility=Collapsed
    iconBrush {
     bIsDynamicallyLoaded=false
     drawAs=Image
     imageType=NoImage
     mirroring=NoMirror
     resourceName=None
     resourceObject=/Game/Athena/Prototype/Textures/T_Placeholder_Generic-S
     tiling=NoTile
     imageSize=<32 32>
     margin=<ltrb 0 0 0 0>
     outlineSettings {
      bUseBrushTransparency=false
      roundingType=HalfHeightRadius
      width=0
      color {colorUseRule=UseColor_Specified, specifiedColor=<rgba 0 0 0 0>}
      cornerRadii=<0 0 0 1>
     }
     tintColor {colorUseRule=UseColor_Specified, specifiedColor=<rgba 1 1 1 1>}
     uVRegion {bIsValid=false, max=<0 0>, min=<0 0>}
    }
   }
   slot {class=BorderSlot, properties {padding=<ltrb 0 0 0 0>}}
  }
  Style_Overlay {
   widgetClassPath=&Overlay
   bInherited=true
   bIsVariable=true
   parent=PanelSwitcher
   widget=Style_Overlay
   slot {class=WidgetSwitcherSlot, properties {verticalAlignment=VAlign_Top}}
  }
  HorizontalBox_1 {
   widgetClassPath=&HorizontalBox
   parent=Style_Overlay
   widget=HorizontalBox_1
   slot {class=OverlaySlot, properties {horizontalAlignment=HAlign_Fill}}
  }
  Spacer_1 {
   widgetClassPath=&Spacer
   parent=HorizontalBox_1
   widget=Spacer_1
   properties {size=<252 1>}
   slot {class=HorizontalBoxSlot, properties {horizontalAlignment=HAlign_Left, verticalAlignment=VAlign_Top, size {sizeRule=Automatic, value=0.2}}}
  }
  Overlay_3 {
   widgetClassPath=&Overlay
   parent=HorizontalBox_1
   widget=Overlay_3
   slot {class=HorizontalBoxSlot, properties {verticalAlignment=VAlign_Top}}
  }
  SizeBox_2 {
   widgetClassPath=&SizeBox
   parent=Overlay_3
   widget=SizeBox_2
   properties {bOverride_HeightOverride=true, bOverride_WidthOverride=true, heightOverride=440, widthOverride=696}
   slot {class=OverlaySlot}
  }
  Image_3 {
   widgetClassPath=&Image
   parent=SizeBox_2
   widget=Image_3
   properties {
    brush {
     bIsDynamicallyLoaded=false
     drawAs=Box
     imageType=NoImage
     mirroring=NoMirror
     resourceName=None
     resourceObject=null
     tiling=NoTile
     imageSize=<32 32>
     margin=<ltrb 0 0 0 0>
     outlineSettings {
      bUseBrushTransparency=false
      roundingType=FixedRadius
      width=4
      color {colorUseRule=UseColor_Specified, specifiedColor=<rgba 0 0.5457 1 1>}
      cornerRadii=<20 20 20 20>
     }
     tintColor {colorUseRule=UseColor_Specified, specifiedColor=<rgba 1 1 1 1>}
     uVRegion {bIsValid=false, max=<0 0>, min=<0 0>}
    }
    colorAndOpacity=<rgba 0.0009 0.006 0.0395 1>
   }
   slot {class=SizeBoxSlot}
  }
  SizeBox_4 {
   widgetClassPath=&SizeBox
   parent=Overlay_3
   widget=SizeBox_4
   properties {bOverride_HeightOverride=true, bOverride_WidthOverride=true, heightOverride=390, widthOverride=650}
   slot {class=OverlaySlot, properties {horizontalAlignment=HAlign_Fill, verticalAlignment=VAlign_Fill}}
  }
  ScrollBoxWords {
   widgetClassPath=&ScrollBox
   bIsVariable=true
   parent=SizeBox_4
   widget=ScrollBoxWords
   properties {bAnimateWheelScrolling=true, scrollBarVisibility=Hidden, scrollWhenFocusChanges=AnimatedScroll}
   slot {class=SizeBoxSlot}
  }
  Overlay_1 {
   widgetClassPath=&Overlay
   parent=ScrollBoxWords
   widget=Overlay_1
   slot {class=ScrollBoxSlot, properties {padding=<ltrb 24 24 24 24>}}
  }
  WordButtons_WBox {
   widgetClassPath=&WrapBox
   bInherited=true
   bIsVariable=true
   parent=Overlay_1
   widget=WordButtons_WBox
   properties {navigation=&self:WidgetTree.WordButtons_WBox.WidgetNavigation_0}
   slot {class=OverlaySlot, properties {horizontalAlignment=HAlign_Fill, verticalAlignment=VAlign_Fill}}
  }
  Spacer_4 {
   widgetClassPath=&Spacer
   parent=HorizontalBox_1
   widget=Spacer_4
   properties {size=<24 1>}
   slot {class=HorizontalBoxSlot, properties {size {sizeRule=Automatic, value=0.05}}}
  }
  VerticalBox_0 {
   widgetClassPath=&VerticalBox
   parent=HorizontalBox_1
   widget=VerticalBox_0
   slot {class=HorizontalBoxSlot, properties {horizontalAlignment=HAlign_Left, verticalAlignment=VAlign_Top, padding=<ltrb 16 0 0 0>}}
  }
  SizeBox_5 {
   widgetClassPath=&SizeBox
   parent=VerticalBox_0
   widget=SizeBox_5
   properties {bOverride_MaxDesiredHeight=true, clipping=ClipToBounds, heightOverride=450, maxDesiredHeight=460}
   slot {class=VerticalBoxSlot, properties {size {sizeRule=Fill, value=1}}}
  }
  StyleButtons_Scrollbox {
   widgetClassPath=&ScrollBox
   bIsVariable=true
   parent=SizeBox_5
   widget=StyleButtons_Scrollbox
   properties {clipping=Inherit, scrollWhenFocusChanges=InstantScroll}
   slot {class=SizeBoxSlot}
  }
  EntryBox_StyleButtons {
   widgetClassPath=&DynamicEntryBox
   bInherited=true
   bIsVariable=true
   parent=StyleButtons_Scrollbox
   widget=EntryBox_StyleButtons
   properties {
    entryBoxType=Vertical
    entryWidgetClass=&/StyleButton.StyleButton_C
    navigation=&self:WidgetTree.EntryBox_StyleButtons.WidgetNavigation_0
    numDesignerPreviewEntries=7
    entrySpacing=<0 6>
   }
   slot {class=ScrollBoxSlot}
  }
  Overlay_105 {
   widgetClassPath=&Overlay
   parent=Overlay_SelectTags
   widget=Overlay_105
   slot {class=OverlaySlot, properties {horizontalAlignment=HAlign_Fill, verticalAlignment=VAlign_Bottom, padding=<ltrb 0 0 0 30>}}
  }
  Button_Cancel {
   widgetClassPath=&IconTextButton
   bIsVariable=true
   parent=Overlay_105
   widget=Button_Cancel
   properties {
    buttonText=Cancel
    contentAlignment=HAlign_Center
    minHeight=48
    minWidth=142
    onButtonBaseClicked=(&self:WidgetTree.Button_Cancel.HandleOnButtonClicked)
    onButtonBaseDoubleClicked=(&self:WidgetTree.Button_Cancel.HandleOnButtonDoubleClicked)
    onButtonBaseHovered=(&self:WidgetTree.Button_Cancel.HandleOnButtonHovered)
    onButtonBaseUnhovered=(&self:WidgetTree.Button_Cancel.HandleOnButtonUnhovered)
    onSelectedChangedBase=(&self:WidgetTree.Button_Cancel.HandleOnSelectedChanged)
    style=/FortUILibrary/Foundation/Buttons/ButtonStyles/Athena/Athena_ButtonStyle_TextOnly_White.Athena_ButtonStyle_TextOnly_White_C
    triggeringInputAction {dataTable=&InputActionDataTable, rowName=GenericBack}
   }
   slot {class=OverlaySlot, properties {horizontalAlignment=HAlign_Right, verticalAlignment=VAlign_Bottom, padding=<ltrb 0 0 60 0>}}
  }
  Button_Help {
   widgetClassPath=&IconTextButton
   bIsVariable=true
   parent=Overlay_105
   widget=Button_Help
   properties {
    buttonText=Help
    contentAlignment=HAlign_Center
    minHeight=48
    minWidth=142
    onButtonBaseClicked=(&self:WidgetTree.Button_Help.HandleOnButtonClicked)
    onButtonBaseDoubleClicked=(&self:WidgetTree.Button_Help.HandleOnButtonDoubleClicked)
    onButtonBaseHovered=(&self:WidgetTree.Button_Help.HandleOnButtonHovered)
    onButtonBaseUnhovered=(&self:WidgetTree.Button_Help.HandleOnButtonUnhovered)
    onSelectedChangedBase=(&self:WidgetTree.Button_Help.HandleOnSelectedChanged)
    style=&Athena_ButtonStyle_TextOnly_White_C
    triggeringInputAction {dataTable=&InputActionDataTable, rowName=GenericLeftStickClick}
   }
   slot {class=OverlaySlot, properties {verticalAlignment=VAlign_Bottom, padding=<ltrb 60 0 0 0>}}
  }
  Button_Style {
   widgetClassPath=&IconTextButton
   bIsVariable=true
   parent=Overlay_105
   widget=Button_Style
   properties {
    bHideInputActionWithKeyboard=true
    buttonText=Apply Style
    contentAlignment=HAlign_Center
    minHeight=48
    minWidth=142
    onButtonBaseClicked=(&self:WidgetTree.Button_Style.HandleOnButtonClicked)
    onButtonBaseDoubleClicked=(&self:WidgetTree.Button_Style.HandleOnButtonDoubleClicked)
    onButtonBaseHovered=(&self:WidgetTree.Button_Style.HandleOnButtonHovered)
    onButtonBaseUnhovered=(&self:WidgetTree.Button_Style.HandleOnButtonUnhovered)
    onSelectedChangedBase=(&self:WidgetTree.Button_Style.HandleOnSelectedChanged)
    style=&Athena_ButtonStyle_TextOnly_White_C
    triggeringInputAction {dataTable=&InputActionDataTable, rowName=GenericFaceButtonLeft}
   }
   slot {class=OverlaySlot, properties {verticalAlignment=VAlign_Bottom, padding=<ltrb 480 0 0 0>}}
  }
  Button_WordButtonLabel {
   widgetClassPath=&IconTextButton
   bIsVariable=true
   parent=Overlay_105
   widget=Button_WordButtonLabel
   properties {
    bIsFocusable=false
    bShouldUseFallbackDefaultInputAction=false
    buttonText=Select
    contentAlignment=HAlign_Center
    minHeight=48
    minWidth=142
    onButtonBaseClicked=(&self:WidgetTree.Button_WordButtonLabel.HandleOnButtonClicked)
    onButtonBaseDoubleClicked=(&self:WidgetTree.Button_WordButtonLabel.HandleOnButtonDoubleClicked)
    onButtonBaseHovered=(&self:WidgetTree.Button_WordButtonLabel.HandleOnButtonHovered)
    onButtonBaseUnhovered=(&self:WidgetTree.Button_WordButtonLabel.HandleOnButtonUnhovered)
    onSelectedChangedBase=(&self:WidgetTree.Button_WordButtonLabel.HandleOnSelectedChanged)
    style=/FortniteCoreUI/InGame/Creative/PropertyEditor/StyleEditor/TextStyles/Athena_ButtonStyle_TextOnly_White_RichText.Athena_ButtonStyle_TextOnly_White_RichText_C
    triggeringInputAction {dataTable=&InputActionDataTable, rowName=GenericForward}
   }
   slot {class=OverlaySlot, properties {verticalAlignment=VAlign_Bottom, padding=<ltrb 250 0 0 0>}}
  }
  SaveAndClose_Button {
   widgetClassPath=&WBP_UIKit_Button_Regular
   bIsVariable=true
   parent=Overlay_105
   widget=SaveAndClose_Button
   properties {
    bIsFocusable=false
    minHeight=56
    minWidth=264
    text="SAVE && CLOSE"
    triggeringInputAction {dataTable=&InputActionDataTable, rowName=GenericFaceButtonTop}
   }
   slot {class=OverlaySlot, properties {horizontalAlignment=HAlign_Center, verticalAlignment=VAlign_Bottom, padding=<ltrb 0 0 0 120>}}
  }
  Button_ClearSelection {
   widgetClassPath=&IconTextButton
   bIsVariable=true
   parent=Overlay_105
   widget=Button_ClearSelection
   properties {
    buttonText=Clear Selection
    contentAlignment=HAlign_Center
    minHeight=48
    minWidth=142
    onButtonBaseClicked=(&self:WidgetTree.Button_ClearSelection.HandleOnButtonClicked)
    onButtonBaseDoubleClicked=(&self:WidgetTree.Button_ClearSelection.HandleOnButtonDoubleClicked)
    onButtonBaseHovered=(&self:WidgetTree.Button_ClearSelection.HandleOnButtonHovered)
    onButtonBaseUnhovered=(&self:WidgetTree.Button_ClearSelection.HandleOnButtonUnhovered)
    onSelectedChangedBase=(&self:WidgetTree.Button_ClearSelection.HandleOnSelectedChanged)
    style=&Athena_ButtonStyle_TextOnly_White_C
    triggeringInputAction {dataTable=&InputActionDataTable, rowName=RightStickAndRightMouseClick}
   }
   slot {class=OverlaySlot, properties {horizontalAlignment=HAlign_Right, verticalAlignment=VAlign_Bottom, padding=<ltrb 0 0 250 0>}}
  }
  ConfirmationWindow {
   widgetClassPath=/Game/UI/Foundation/Modals/ConfirmationWindow/ConfirmationWindow
   bInherited=true
   bIsVariable=true
   parent=Overlay_SelectTags
   widget=ConfirmationWindow
   properties {
    bIsModal=true
    bSetVisibilityOnActivated=true
    bSetVisibilityOnDeactivated=true
    visibility=Collapsed
    primaryContentSetup {bEnableNewTextChat=false, bShowBottomBar=false, bShowChatWidget=false, bShowTopBar=false, bTabSupportsInterface=false}
   }
   slot {class=OverlaySlot, properties {horizontalAlignment=HAlign_Fill, verticalAlignment=VAlign_Fill}}
  }
 ]
}
