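# Creative property editor row for advanced string options, editing text through a modal text and style editor instead of inline.
# Edit Text async-queues the modal, passing styles, text and input settings from the collapsed filtered field, then unbinds when it deactivates.
# Setting text resizes that field's character limit to fit, so the original limit is cached for the editor.
# Row selection toggles a rich text preview dialog after a short delay.
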WidgetBlueprint /FortniteCoreUI/InGame/Creative/PropertyEditor/EditWidgets/CreativePropertyEditWidgetBase_String_Advanced
graphs [
 BP_GetDesiredFocusTarget {nodes [
   Get Desired Focus Target()
   > n1 = Branch(Condition=AllowModalFocus!(self:Creative Property Edit Widget String Base Object Reference) -> (ReturnValue:bool))
   n2 = Get RichTextEditorWidget() -> (RichTextEditorWidget:Text And Style Editor Object Reference)
   > n3 = Is Valid(InputObject=@n2)
   > n4 = Branch(Condition=IsActivated!(self:Common Activatable Widget Object Reference=@n2) -> (ReturnValue:bool))
   > Return Node!(
    execute
    ReturnValue:Widget Object Reference=GetDesiredFocusTarget!(self:Common Activatable Widget Object Reference=@n2) -> (ReturnValue:Widget Object Reference)
   )
   Return Node(execute=[@n4.else, @n3.Is Not Valid, @n1.else], ReturnValue=Self-Reference())
  ]}
 EventGraph {nodes [
   n1 = "On Text Changed (FilteredTextEntry_Widget)"!() -> (OutputDelegate, then, Text:Text) {is=title K2Node_ComponentBoundEvent}
   > SetIsModified!(execute, self:Creative Property Edit Widget Base Object Reference, bModified:bool=false) -> (then)
   > SetIsModified(bModified=NOT Boolean(A=TextIsEmpty(InText=@n1.Text)))
   > RequestCloseDialog!(execute, self:Creative Property Edit Widget Base Object Reference, bSetFocusToOptionsList:bool=false) -> (then)
   "On Button Clicked (Edit_button)"!() -> (OutputDelegate, then, Button:Common Button Legacy Object Reference) {is=title K2Node_ComponentBoundEvent}
   > n2 = Cast To TextAndStyleEditor!(execute, Object=Get RichTextEditorWidget()) -> (then, CastFailed, AsText and Style Editor:Text And Style Editor Object Reference) {is=title K2Node_DynamicCast}
   n3 = QueueModalWidgetSoftClassAsync!(
    execute:Exec=@n2.CastFailed
    SoftWidgetClass:Common Activatable Widget Soft Class Reference=/FortniteCoreUI/InGame/Creative/PropertyEditor/StyleEditor/TextAndStyleEditor.TextAndStyleEditor_C
   ) -> (then:Exec, OnModalWidgetQueued:Exec, ActivatableWidget:Common Activatable Widget Object Reference) {is=title K2Node_AsyncAction}
   n4 = Cast To TextAndStyleEditor(execute=@n3.OnModalWidgetQueued, Object=@n3)
   > n5 = Set RichTextEditorWidget(RichTextEditorWidget:Text And Style Editor Object Reference=@n4) -> (Output_Get:Text And Style Editor Object Reference)
   > Setup Rich Text Styles!(
    execute
    self:Text And Style Editor Object Reference=@n5
    Owning Device Object:Object Reference=GetInitializingObject!(self:Creative Property Edit Widget User Option Base Object Reference) -> (ReturnValue:Object Reference)
   ) -> (then)
   > SyncTextSettings!(execute, self, RichTextEditor:Text And Style Editor Object Reference=@n5) -> (then)
   > Initialize Text!(
    execute
    self:Text And Style Editor Object Reference=@n5
    inText:Text=Get Filtered Text Entry Value!() -> (ReturnValue:Text) {is=title K2Node_CallFunction}
   ) -> (then)
   n6 = Create Event()
   n7 = Get RichTextEditorWidget()
   > n8 = Bind Event to On Text Change Confirmed(self:Text And Style Editor Object Reference=@n7, Delegate=@n6) {is=title K2Node_AddDelegate}
   n9 = HandleOnEditorWidgetDeactivated() {is=title K2Node_CustomEvent}
   Bind Event to On Widget Deactivated(execute=@n8.then, self:Common Activatable Widget Object Reference=@n7, Delegate=@n9) {is=title K2Node_AddDelegate}
   > SetIsEnabled(self=Get Edit_button() -> (Edit_button:Icon Text Button Object Reference), bInIsEnabled=false)
   "On Button Clicked (PreviewColor_Button)"!() -> (OutputDelegate, then, Button:Common Button Legacy Object Reference) {is=title K2Node_ComponentBoundEvent}
   n10 = Get RichTextEditorPreviewWidget() -> (RichTextEditorPreviewWidget:Creative Edit Dialog Widget Rich Text Editor Object Reference)
   > Is Valid(InputObject=@n10)
   > n11 = GetCurrentBackgroundColor!(execute, self:Creative Edit Dialog Widget Rich Text Editor Object Reference=@n10) -> (then, OutIsUsingLightColor:bool)
   > UpdateColor!(execute, self:Creative Edit Dialog Widget Rich Text Editor Object Reference=@n10, InUseLightColor:bool=NOT Boolean(A=@n11)) -> (then)
   n12 = Update Text!() -> (OutputDelegate, then, InString:String, ShouldUpdate:bool=~) {is=title K2Node_CustomEvent}
   > Branch(Condition=@n12.ShouldUpdate)
   > Set Filtered Text Entry Value!(execute, self, InText:Text="To Text (String)"(InString=@n12.InString)) -> (then) {is=title K2Node_CallFunction}
   n13 = Handle_SetColorButtonVisibility!() -> (OutputDelegate, then, bNewInputType:ECommonInputType Enum) {is=title K2Node_CustomEvent}
   > Set RenderOpacity!(
    execute
    RenderOpacity:float=Select!(MouseAndKeyboard:float=1, Gamepad:float=0, Touch:float=1, Count:float=0, Index:ECommonInputType Enum=@n13.bNewInputType) -> (ReturnValue:float)
    self:Widget Object Reference=Get PreviewColor_Button() -> (PreviewColor_Button:Icon Text Button Object Reference)
   ) -> (then, Output_Get:float)
   > SetVisibility(
    self=Get Color_Icons_Inner_Overlay() -> (Color_Icons_Inner_Overlay:Overlay Object Reference)
    InVisibility=Select(
     MouseAndKeyboard:ESlateVisibility Enum=SelfHitTestInvisible
     Gamepad:ESlateVisibility Enum=Collapsed
     Touch:ESlateVisibility Enum=HitTestInvisible
     Count:ESlateVisibility Enum=Visible
     Index=@n13.bNewInputType
    ) -> (ReturnValue:ESlateVisibility Enum)
   )
   n14 = Event BP_OnModified!() -> (OutputDelegate, then, bModified:bool) {is=title K2Node_Event}
   > UpdateModifiedState!(execute, self, isModifiedFromDefaultValue:bool=@n14.bModified) -> (then)
   Event Construct()
   n15 = Get CommonInputSubsystem()
   > Bind Event to On Input Method Changed(self=@n15, Delegate=@n13)
   n16 = GetCurrentInputType(self=@n15)
   > Set RenderOpacity(
    RenderOpacity=Select(MouseAndKeyboard:float=1, Gamepad:float=0, Touch:float=1, Count:float=0, Index=@n16) -> (ReturnValue:float)
    self=Get PreviewColor_Button()
   )
   > SetVisibility(
    self=Get Color_Icons_Inner_Overlay()
    InVisibility=Select(
     MouseAndKeyboard:ESlateVisibility Enum=SelfHitTestInvisible
     Gamepad:ESlateVisibility Enum=Collapsed
     Touch:ESlateVisibility Enum=HitTestInvisible
     Count:ESlateVisibility Enum=Visible
     Index=@n16
    ) -> (ReturnValue:ESlateVisibility Enum)
   )
   Event Destruct()
   > Unbind Event from On Input Method Changed(self=Get CommonInputSubsystem(), Delegate=@n13)
   Event BP_PostInitEditWidget() {is=title K2Node_Event}
   > SaveEntryCharLimit()
   > "SetText (Text)"(
    self=Get TextBlock_OptionDisplayName_2() -> (TextBlock_OptionDisplayName_2:Common Text Object Reference)
    InText=GetEntryDisplayName!(self:Creative Property Edit Widget Base Object Reference) -> (ReturnValue:Text)
   )
   > n17 = Cast To CreativeEditDialogWidget_RichTextEditor!(execute, Object=Get DialogWidget() -> (DialogWidget:Fort Creative Property Editor Dialog Base Object Reference)) -> (then, CastFailed, AsCreative Edit Dialog Widget Rich Text Editor:Creative Edit Dialog Widget Rich Text Editor Object Reference) {is=title K2Node_DynamicCast}
   > Set RichTextEditorPreviewWidget(RichTextEditorPreviewWidget:Creative Edit Dialog Widget Rich Text Editor Object Reference=@n17) -> (Output_Get:Creative Edit Dialog Widget Rich Text Editor Object Reference)
   > UpdateTextPreviewWidget!(execute, self, OpenDialog:bool=false) -> (then)
   > SetVisibility(self=Get DialogWidget(), InVisibility=Collapsed)
   n18 = Event OnEditWidgetSelectedChanged!() -> (OutputDelegate, then, bSelected:bool) {is=title K2Node_Event}
   > Delay(Duration=0.1)
   > UpdateTextPreviewWidget(OpenDialog=@n18.bSelected)
   SetIsEnabled(execute=@n20.then, self=Get Edit_button(), bInIsEnabled=true)
   n19 = Get RichTextEditorWidget()
   Unbind Event from On Text Change Confirmed(execute=@n9.then, self:Text And Style Editor Object Reference=@n19, Delegate=@n6) {is=title K2Node_RemoveDelegate}
   > Unbind Event from On Widget Deactivated(self:Common Activatable Widget Object Reference=@n19, Delegate=@n9) {is=title K2Node_RemoveDelegate}
   > n20 = Set RichTextEditorWidget()
  ]}
 Get Filtered Text Entry Value {nodes [
   Get Filtered Text Entry Value() {is=title K2Node_FunctionEntry}
   > n1 = Cast To WBP_UIKit_FilteredInputField!(execute, Object=Get FilteredTextEntry_Widget() -> (FilteredTextEntry_Widget:Fort Editable Filtered Counted Text Box Object Reference)) -> (then, CastFailed, AsWBP UIKit Filtered Input Field:WBP UIKit Filtered Input Field Object Reference) {is=title K2Node_DynamicCast}
   > n2 = GetCurrentText!(execute, self:WBP UIKit Filtered Input Field Object Reference=@n1) -> (then, Text:Text)
   > Return Node(ReturnValue:Text=@n2)
  ]}
 HandlesCharLimitValidation {nodes [
   HandlesCharLimitValidation() {is=title K2Node_FunctionEntry}
   > Return Node(ReturnValue:bool=true)
  ]}
 OnFocusReceived {nodes [
   OnFocusReceived!() -> (then, MyGeometry:Geometry Structure, InFocusEvent:Focus Event Structure) {is=title K2Node_FunctionEntry}
   n1 = Get FilteredTextEntry_Widget()
   > CenterWidget!(execute, ToCenter:Widget Object Reference=@n1) -> (then)
   > SetKeyboardFocus(self:Widget Object Reference=@n1)
   > Return Node(ReturnValue:Event Reply Structure=Handled())
  ]}
 SaveEntryCharLimit {nodes [
   SaveEntryCharLimit() {is=title K2Node_FunctionEntry}
   > Set OriginalEntryCharLimit(OriginalEntryCharLimit:int=GetEntryCharLimit!(self:Fort Editable Filtered Counted Text Box Object Reference=Get FilteredTextEntry_Widget()) -> (ReturnValue:int=0)) -> (Output_Get:int=0)
  ]}
 Set Filtered Text Entry Value {nodes [
   n1 = Set Filtered Text Entry Value!() -> (then, InText:Text) {is=title K2Node_FunctionEntry}
   n2 = Get FilteredTextEntry_Widget()
   > SetEntryCharLimit!(execute, self:Fort Editable Filtered Counted Text Box Object Reference=@n2, InCharLimit:int=Len(S="To String (Text)"(InText=@n1))) -> (then)
   > SetTextWithForcedSanitization!(execute, self:Fort Editable Filtered Counted Text Box Object Reference=@n2, Text:"Text (by ref)"=@n1) -> (then)
   > Is Valid(InputObject=Get RichTextEditorPreviewWidget())
   > Update Text!(
    execute
    self:Creative Edit Dialog Widget Rich Text Editor Object Reference=Get RichTextEditorPreviewWidget()
    InText:"Text (by ref)"=Get Filtered Text Entry Value() {is=title K2Node_CallFunction}
   ) -> (then) {is=title K2Node_CallFunction}
  ]}
 SyncTextSettings {nodes [
   n1 = SyncTextSettings!() -> (then, RichTextEditor:Text And Style Editor Object Reference) {is=title K2Node_FunctionEntry}
   > SetEntryCharLimit(self:Fort Creative Rich Text Editor Widget Object Reference=@n1, InCharLimit=Get OriginalEntryCharLimit() -> (OriginalEntryCharLimit:int=0))
   n2 = Get FilteredTextEntry_Widget()
   > SetAllowLineBreaks!(
    execute
    self:Fort Creative Rich Text Editor Widget Object Reference=@n1
    InAllowLineBreaks:bool=GetAllowLineBreaks!(self:Fort Editable Filtered Counted Text Box Object Reference=@n2) -> (ReturnValue:bool)
   ) -> (then)
   > SetAllowStyleEditing!(
    execute
    self:Fort Creative Rich Text Editor Widget Object Reference=@n1
    InAllowStyleEditing:bool=GetAllowStyleEditing!(self:Fort Editable Filtered Counted Text Box Object Reference=@n2) -> (ReturnValue:bool)
   ) -> (then)
   > SetDeviceInfo!(
    execute
    self:Text And Style Editor Object Reference=@n1
    DeviceName:Text=RequestDeviceName!(self:Creative Property Edit Widget Base Object Reference) -> (ReturnValue:"Text (by ref)")
    DeviceFieldName:Text=Get Text(self:Text Object Reference=Get TextBlock_OptionDisplayName_2())
   ) -> (then)
  ]}
 UpdateModifiedState {nodes [
   n1 = UpdateModifiedState!() -> (then, isModifiedFromDefaultValue:bool=~) {is=title K2Node_FunctionEntry}
   > SetVisibility(
    self=Get Image_Modified() -> (Image_Modified:Image Object Reference)
    InVisibility=Select!(Option 0:ESlateVisibility Enum=Hidden, Option 1:ESlateVisibility Enum=SelfHitTestInvisible, Index:bool=@n1) -> (ReturnValue)
   )
  ]}
 UpdateTextPreviewWidget {nodes [
   n1 = UpdateTextPreviewWidget!() -> (then, OpenDialog:bool=~) {is=title K2Node_FunctionEntry}
   > Is Valid(InputObject=Get RichTextEditorPreviewWidget())
   > Update Text(self=Get RichTextEditorPreviewWidget(), InText=Get Filtered Text Entry Value())
   > n2 = Branch(Condition=@n1)
   n3 = Get RichTextEditorPreviewWidget()
   > SetVisibility(self=@n3, InVisibility=Visible)
   > Open(self:Fort Creative Property Editor Dialog Base Object Reference=@n3)
   > UpdateColor(self=@n3, InUseLightColor=false)
   Close(execute=@n2.else, self:Fort Creative Property Editor Dialog Base Object Reference=@n3)
   > SetVisibility(self=@n3, InVisibility=Collapsed)
  ]}
]
variables [
 OriginalEntryCharLimit {type=integer}
 RichTextEditorPreviewWidget {type=CreativeEditDialogWidget_RichTextEditor_C, default=null}
 RichTextEditorWidget {type=TextAndStyleEditor_C, default=null}
]
widgets {
 info {
  inheritedWidgetCount=2
  namedSlotCount=2
  parentClass=/Script/FortniteUI.CreativePropertyEditWidgetStringBase
  rootWidgetClass=&Overlay
  widgetCount=22
 }
 widgets [
  Overlay_25 {widgetClassPath=&Overlay, widget=Overlay_25}
  SizeBox_3 {
   widgetClassPath=&SizeBox
   parent=Overlay_25
   widget=SizeBox_3
   properties {bOverride_HeightOverride=true, heightOverride=82}
   slot {class=OverlaySlot, properties {horizontalAlignment=HAlign_Fill, verticalAlignment=VAlign_Center}}
  }
  Image_Selection_Hover {
   widgetClassPath=&Image
   parent=SizeBox_3
   widget=Image_Selection_Hover
   properties {
    visibility=HitTestInvisible
    brush {
     bIsDynamicallyLoaded=false
     drawAs=Image
     imageType=NoImage
     mirroring=NoMirror
     resourceName=None
     resourceObject=/Game/Athena/UI/InGame/Creative/Art/M_SimpleSkewBox_CreativeQuickMenuSelection
     tiling=NoTile
     imageSize=<32 82>
     margin=<ltrb 0 0 0 0>
     outlineSettings {
      bUseBrushTransparency=false
      roundingType=HalfHeightRadius
      width=0
      color {colorUseRule=UseColor_Specified, specifiedColor=<rgba 0 0 0 0>}
      cornerRadii=<0 0 0 0>
     }
     tintColor {colorUseRule=UseColor_Specified, specifiedColor=<rgba 1 1 1 1>}
     uVRegion {bIsValid=false, max=<0 0>, min=<0 0>}
    }
    renderTransform {angle=0, scale=<1.31 1.3>, shear=<0 0>, translation=<0 0>}
   }
   slot {class=SizeBoxSlot}
  }
  SizeBox_0 {
   widgetClassPath=&SizeBox
   parent=Overlay_25
   widget=SizeBox_0
   properties {bOverride_MinDesiredHeight=true, minDesiredHeight=82}
   slot {class=OverlaySlot, properties {horizontalAlignment=HAlign_Fill, verticalAlignment=VAlign_Center, padding=<ltrb 10 0 16 0>}}
  }
  Overlay_0 {widgetClassPath=&Overlay, parent=SizeBox_0, widget=Overlay_0, slot {class=SizeBoxSlot}}
  HorizontalBox_2 {
   widgetClassPath=&HorizontalBox
   parent=Overlay_0
   widget=HorizontalBox_2
   slot {class=OverlaySlot, properties {horizontalAlignment=HAlign_Fill, verticalAlignment=VAlign_Fill}}
  }
  Image_Modified {
   widgetClassPath=&Image
   bIsVariable=true
   parent=HorizontalBox_2
   widget=Image_Modified
   properties {
    visibility=Hidden
    brush {
     bIsDynamicallyLoaded=false
     drawAs=Image
     imageType=NoImage
     mirroring=NoMirror
     resourceName=None
     resourceObject=/Game/Athena/UI/InGame/Creative/PropertyEditor/Art/Icon-PropertyEditor-Modified
     tiling=NoTile
     imageSize=<20 20>
     margin=<ltrb 0 0 0 0>
     outlineSettings {
      bUseBrushTransparency=false
      roundingType=HalfHeightRadius
      width=0
      color {colorUseRule=UseColor_Specified, specifiedColor=<rgba 0 0 0 0>}
      cornerRadii=<0 0 0 0>
     }
     tintColor {colorUseRule=UseColor_Specified, specifiedColor=<rgba 1 1 1 1>}
     uVRegion {bIsValid=false, max=<0 0>, min=<0 0>}
    }
    colorAndOpacity=<rgba 0 0.0887 0.9823 1>
   }
   slot {class=HorizontalBoxSlot, properties {horizontalAlignment=HAlign_Left, verticalAlignment=VAlign_Center, padding=<ltrb 0 0 4 0>}}
  }
  SizeBox_2 {
   widgetClassPath=&SizeBox
   parent=HorizontalBox_2
   widget=SizeBox_2
   properties {widthOverride=900}
   slot {class=HorizontalBoxSlot, properties {verticalAlignment=VAlign_Center, padding=<ltrb 0 0 12 0>, size {sizeRule=Fill, value=0.5}}}
  }
  TextBlock_OptionDisplayName_2 {
   widgetClassPath=&CommonTextBlock
   bIsVariable=true
   parent=SizeBox_2
   widget=TextBlock_OptionDisplayName_2
   properties {
    autoWrapText=true
    scrollStyle=&TextScrollStyle-Base_C
    text="[Parameter name] "
    colorAndOpacity {colorUseRule=UseColor_Specified, specifiedColor=<rgba 0 0.0887 0.9823 1>}
    font {
     bForceMonospaced=false
     bMaterialIsStencil=false
     fontMaterial=null
     fontObject=&BurbankSmall
     letterSpacing=0
     monospacedWidth=1
     size=18
     skewAmount=0
     typefaceFontName=Bold
     outlineSettings {
      bApplyOutlineToDropShadows=false
      bMiteredCorners=false
      bSeparateFillAlpha=false
      outlineMaterial=null
      outlineSize=0
      outlineColor=<rgba 0.0052 0.0052 0.0056 0.2>
     }
    }
   }
   slot {class=SizeBoxSlot, properties {verticalAlignment=VAlign_Center}}
  }
  SizeBox_1 {
   widgetClassPath=&SizeBox
   parent=HorizontalBox_2
   widget=SizeBox_1
   properties {renderOpacity=0, visibility=Collapsed, widthOverride=715.3185}
   slot {class=HorizontalBoxSlot, properties {padding=<ltrb 0 4 32 0>, size {sizeRule=Fill, value=0.5}}}
  }
  OptionSelectionArea_2 {widgetClassPath=&HorizontalBox, bIsVariable=true, parent=SizeBox_1, widget=OptionSelectionArea_2, slot {class=SizeBoxSlot}}
  FilteredTextEntry_Widget {
   widgetClassPath=/FortUILibrary/UIKit/InputField/WBP_UIKit_FilteredInputField
   bInherited=true
   bIsVariable=true
   parent=OptionSelectionArea_2
   widget=FilteredTextEntry_Widget
   properties {bIsFocusable=true, foregroundColor {colorUseRule=UseColor_Specified, specifiedColor=<rgba 0 0 0 1>}}
   slot {class=HorizontalBoxSlot, properties {size {sizeRule=Fill, value=1}}}
  }
  PreviewColor_Button {
   widgetClassPath=&IconTextButton
   bIsVariable=true
   parent=HorizontalBox_2
   widget=PreviewColor_Button
   properties {
    bIsEnabled=false
    buttonText=""
    onButtonBaseClicked=(&self:WidgetTree.PreviewColor_Button.HandleOnButtonClicked)
    onButtonBaseDoubleClicked=(&self:WidgetTree.PreviewColor_Button.HandleOnButtonDoubleClicked)
    onButtonBaseHovered=(&self:WidgetTree.PreviewColor_Button.HandleOnButtonHovered)
    onButtonBaseUnhovered=(&self:WidgetTree.PreviewColor_Button.HandleOnButtonUnhovered)
    onSelectedChangedBase=(&self:WidgetTree.PreviewColor_Button.HandleOnSelectedChanged)
    style=/FortUILibrary/Foundation/Buttons/ButtonStyles/Athena/Athena_ButtonStyle_AngledBlueMenuButton_Small.Athena_ButtonStyle_AngledBlueMenuButton_Small_C
    visibility=Collapsed
    triggeringInputAction {dataTable=&InputActionDataTable, rowName=GenericForward}
   }
   slot {class=HorizontalBoxSlot, properties {verticalAlignment=VAlign_Center, padding=<ltrb 0 0 24 0>}}
  }
  Color_Icons_Overlay {widgetClassPath=&Overlay, namedSlotHost=PreviewColor_Button, widget=Color_Icons_Overlay}
  Color_Icons_Inner_Overlay {
   widgetClassPath=&Overlay
   bIsVariable=true
   parent=Color_Icons_Overlay
   widget=Color_Icons_Inner_Overlay
   properties {renderTransform {angle=0, scale=<1 1>, shear=<0 0>, translation=<0 8>}}
   slot {class=OverlaySlot}
  }
  CommonBorder_47 {
   widgetClassPath=&CommonBorder
   parent=Color_Icons_Inner_Overlay
   widget=CommonBorder_47
   properties {
    background {
     bIsDynamicallyLoaded=false
     drawAs=RoundedBox
     imageType=NoImage
     mirroring=NoMirror
     resourceName=None
     resourceObject=null
     tiling=NoTile
     imageSize=<32 32>
     margin=<ltrb 0 0 0 0>
     outlineSettings {
      bUseBrushTransparency=false
      roundingType=FixedRadius
      width=1
      color {colorUseRule=UseColor_Specified, specifiedColor=<rgba 1 1 1 1>}
      cornerRadii=<2 2 2 2>
     }
     tintColor {colorUseRule=UseColor_Specified, specifiedColor=<rgba 0.0833 0.0833 0.0833 1>}
     uVRegion {bIsValid=false, max=<0 0>, min=<0 0>}
    }
    padding=<ltrb 12 12 12 12>
   }
   slot {class=OverlaySlot}
  }
  CommonBorder_108 {
   widgetClassPath=&CommonBorder
   parent=Color_Icons_Inner_Overlay
   widget=CommonBorder_108
   properties {
    background {
     bIsDynamicallyLoaded=false
     drawAs=RoundedBox
     imageType=NoImage
     mirroring=NoMirror
     resourceName=None
     resourceObject=null
     tiling=NoTile
     imageSize=<32 32>
     margin=<ltrb 0 0 0 0>
     outlineSettings {
      bUseBrushTransparency=false
      roundingType=FixedRadius
      width=1
      color {colorUseRule=UseColor_Specified, specifiedColor=<rgba 0 0 0 1>}
      cornerRadii=<2 2 2 2>
     }
     tintColor {colorUseRule=UseColor_Specified, specifiedColor=<rgba 1 1 1 1>}
     uVRegion {bIsValid=false, max=<0 0>, min=<0 0>}
    }
    padding=<ltrb 12 12 12 12>
   }
   slot {class=OverlaySlot, properties {padding=<ltrb 8 8 0 0>}}
  }
  Edit_button {
   widgetClassPath=&IconTextButton
   bIsVariable=true
   parent=HorizontalBox_2
   widget=Edit_button
   properties {
    buttonText=Edit Text
    minWidth=256
    onButtonBaseClicked=(&self:WidgetTree.Edit_button.HandleOnButtonClicked)
    onButtonBaseDoubleClicked=(&self:WidgetTree.Edit_button.HandleOnButtonDoubleClicked)
    onButtonBaseHovered=(&self:WidgetTree.Edit_button.HandleOnButtonHovered)
    onButtonBaseUnhovered=(&self:WidgetTree.Edit_button.HandleOnButtonUnhovered)
    onSelectedChangedBase=(&self:WidgetTree.Edit_button.HandleOnSelectedChanged)
    style=&Athena_ButtonStyle_AngledBlueMenuButton_Small_C
    triggeringInputAction {dataTable=&InputActionDataTable, rowName=Confirm}
   }
   slot {class=HorizontalBoxSlot, properties {horizontalAlignment=HAlign_Right, padding=<ltrb 0 24 32 24>}}
  }
  SaveSpinner_2 {
   widgetClassPath=&Image
   bIsVariable=true
   parent=Overlay_25
   widget=SaveSpinner_2
   properties {
    visibility=Collapsed
    brush {
     bIsDynamicallyLoaded=false
     drawAs=Image
     imageType=NoImage
     mirroring=NoMirror
     resourceName=None
     resourceObject=/FortUILibrary/UIKit/Throbber/MI_UIKit_Color_Throbber
     tiling=NoTile
     imageSize=<24 24>
     margin=<ltrb 0 0 0 0>
     outlineSettings {
      bUseBrushTransparency=false
      roundingType=HalfHeightRadius
      width=0
      color {colorUseRule=UseColor_Specified, specifiedColor=<rgba 0 0 0 0>}
      cornerRadii=<0 0 0 0>
     }
     tintColor {colorUseRule=UseColor_Specified, specifiedColor=<rgba 1 1 1 1>}
     uVRegion {bIsValid=false, max=<0 0>, min=<0 0>}
    }
   }
   slot {class=OverlaySlot, properties {verticalAlignment=VAlign_Center, padding=<ltrb 10 -24 0 0>}}
  }
  SizeBox_ExpanderOpen_2 {
   widgetClassPath=&SizeBox
   bIsVariable=true
   parent=Overlay_25
   widget=SizeBox_ExpanderOpen_2
   properties {bOverride_HeightOverride=true, bOverride_WidthOverride=true, heightOverride=42, visibility=Collapsed, widthOverride=42}
   slot {class=OverlaySlot, properties {horizontalAlignment=HAlign_Right, verticalAlignment=VAlign_Center}}
  }
  Button_Collapse {
   widgetClassPath=/FortUILibrary/Foundation/Buttons/PanelButton
   bInherited=true
   bIsVariable=true
   parent=SizeBox_ExpanderOpen_2
   widget=Button_Collapse
   properties {
    onButtonBaseClicked=(&self:WidgetTree.Button_Collapse.HandleOnButtonClicked)
    onButtonBaseDoubleClicked=(&self:WidgetTree.Button_Collapse.HandleOnButtonDoubleClicked)
    onButtonBaseHovered=(&self:WidgetTree.Button_Collapse.HandleOnButtonHovered)
    onButtonBaseUnhovered=(&self:WidgetTree.Button_Collapse.HandleOnButtonUnhovered)
    onSelectedChangedBase=(&self:WidgetTree.Button_Collapse.HandleOnSelectedChanged)
    style=/FortUILibrary/Foundation/Buttons/ButtonStyles/ButtonStyle-Mobile.ButtonStyle-Mobile_C
    visibility=Collapsed
    inputActionRenderTranslation=<-38 -4>
    inputActionRimBrushSize=<32 32>
    triggeringInputAction {dataTable=/FortUILibrary/Input/CreativeInputActionDataTable, rowName=CollapseEditWidget}
   }
   slot {class=SizeBoxSlot}
  }
  Image_ExpanderArrow_2 {
   widgetClassPath=&Image
   bIsVariable=true
   namedSlotHost=Button_Collapse
   widget=Image_ExpanderArrow_2
   properties {
    brush {
     bIsDynamicallyLoaded=false
     drawAs=Image
     imageType=NoImage
     mirroring=NoMirror
     resourceName=None
     resourceObject=/Game/UI/SocialPanel/Art/Textures/T_UI_Icon_TreeArrow_64
     tiling=NoTile
     imageSize=<32 32>
     margin=<ltrb 0 0 0 0>
     outlineSettings {
      bUseBrushTransparency=false
      roundingType=HalfHeightRadius
      width=0
      color {colorUseRule=UseColor_Specified, specifiedColor=<rgba 0 0 0 0>}
      cornerRadii=<0 0 0 0>
     }
     tintColor {colorUseRule=UseColor_Specified, specifiedColor=<rgba 1 1 1 1>}
     uVRegion {bIsValid=false, max=<0 0>, min=<0 0>}
    }
    renderTransform {angle=-90, scale=<1 1>, shear=<0 0>, translation=<0 0>}
   }
  }
 ]
}
